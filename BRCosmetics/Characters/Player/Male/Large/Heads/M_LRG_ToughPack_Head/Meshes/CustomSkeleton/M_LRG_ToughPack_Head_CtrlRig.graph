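# Control Rig for the large male ToughPack head cosmetic, procedurally posing its sleeping bag, face and cover bones from head motion.
# Head twist against head_null partially rotates sleepingBag_head and drives remapped weights that offset face_head and the cover IK target.
# Parent constraints blend eyes and brows between bag and face, while belts aim at lid targets.
# Stretchy two-bone IK bends the cover, using a setup-computed pole vector that follows spine_05.
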ControlRigBlueprint /BRCosmetics/Characters/Player/Male/Large/Heads/M_LRG_ToughPack_Head/Meshes/CustomSkeleton/M_LRG_ToughPack_Head_CtrlRig
graphs [
 New Function_ContainedGraph {nodes [
   n1 = RigVMFunctionEntryNode()
   RigVMFunction_ControlFlowBranch = &RigVMFunction_ControlFlowBranch(ExecuteContext=@n1.ExecuteContext, Condition=@n1.debug)
   RigVMFunctionReturnNode(ExecuteContext=@RigVMFunction_ControlFlowBranch.Completed)
   RigUnit_ItemArray = &RigUnit_ItemArray(Items=((Type=Bone,Name="neck_01"),(Type=Bone,Name="neck_02"),(Type=Bone,Name="head")))
   DISPATCH_RigVMDispatch_ArrayIterator = RigVMDispatchNode(Array=@RigUnit_ItemArray.Items, ExecuteContext=@RigVMFunction_ControlFlowBranch.True) {template="DISPATCH_RigVMDispatch_ArrayIterator(in Array,out Element,out Index,out Count,out Ratio)"}
   Concat = /Script/RigVM.RigVMFunction_NameConcat(B=_ctrl, A=@DISPATCH_RigVMDispatch_ArrayIterator.Element.Name) {is=struct RigVMUnitNode, template="Concat::Execute(in A,in B,out Result)"}
   > &RigUnit_SetTransform(
    Item=(Type=Bone,Name="spine_01")
    Weight=1
    bInitial=False
    Item.Name=@DISPATCH_RigVMDispatch_ArrayIterator.Element.Name
    Value=&RigUnit_GetTransform(Item=(Type=Control,Name="spine_01_ctrl"), bInitial=False, Item.Name=@Concat.Result).Transform
   )
  ]}
 New Function_ContainedGraph {nodes [
   n1 = RigVMFunctionEntryNode()
   RigVMFunction_MathBoolToInteger = &RigVMFunction_MathBoolToInteger(Value=@n1.onlyTwist)
   RigVMFunction_MathQuaternionSwingTwist = &RigVMFunction_MathQuaternionSwingTwist(
    TwistAxis=@n1.twistAxis
    Input=&RigUnit_GetRelativeTransformForItem(
     Child=(Type=Bone,Name="None")
     Parent=(Type=Bone,Name="None")
     bChildInitial=False
     bParentInitial=False
     Child.Name=@n1.child
     Parent.Name=@n1.parent
     Parent.Type=@n1.parentType
    ).RelativeTransform.Rotation
   )
   RigVMFunction_MathQuaternionToEuler = &RigVMFunction_MathQuaternionToEuler(
    Value=RigVMDispatchNode(
     Values=((),())
     Index=@RigVMFunction_MathBoolToInteger.Result
     Values.0=@RigVMFunction_MathQuaternionSwingTwist.Swing
     Values.1=@RigVMFunction_MathQuaternionSwingTwist.Twist
    ) {template="DISPATCH_RigVMDispatch_SelectInt32(in Index,in Values,out Result)"}.Result
    RotationOrder=@n1.Rotation Order
   )
   > DISPATCH_RigVMDispatch_Print = RigVMDispatchNode(ScreenColor=(R=1,G=1,B=1,A=1), ScreenDuration=1.5, Enabled=@n1.debugRotation, Value=@RigVMFunction_MathQuaternionToEuler.Result) {template="DISPATCH_RigVMDispatch_Print(in Prefix,in Value,in Enabled,in ScreenDuration,in ScreenColor)"}
   Clamp = &RigVMFunction_MathIntClamp(Value=@n1.primAxis)
   DISPATCH_RigVMDispatch_SelectInt32 = RigVMDispatchNode(
    Values=(0,0,0)
    Index=@Clamp.Result
    Values.0=@RigVMFunction_MathQuaternionToEuler.Result.X
    Values.1=@RigVMFunction_MathQuaternionToEuler.Result.Y
    Values.2=@RigVMFunction_MathQuaternionToEuler.Result.Z
   ) {template="DISPATCH_RigVMDispatch_SelectInt32(in Index,in Values,out Result)"}
   Remap = &RigVMFunction_MathDoubleRemap(
    TargetMaximum=1
    TargetMinimum=0
    bClamp=False
    Value=@DISPATCH_RigVMDispatch_SelectInt32.Result
    SourceMaximum=@n1.max
    SourceMinimum=@n1.min
   )
   Multiply = &RigVMFunction_MathDoubleMul(A=&RigVMFunction_MathDoubleClamp(Value=@Remap.Result).Result, B=@n1.premult)
   > RigVMFunctionReturnNode(outputWeight=@Multiply.Result)
  ]}
 New Function_ContainedGraph {nodes [
   n1 = RigVMFunctionEntryNode()
   RigVMFunction_MathBoolToInteger = &RigVMFunction_MathBoolToInteger(Value=@n1.onlyTwist)
   RigVMFunction_MathQuaternionSwingTwist = &RigVMFunction_MathQuaternionSwingTwist(
    TwistAxis=@n1.twistAxis
    Input=&RigUnit_GetRelativeTransformForItem(bChildInitial=False, bParentInitial=False, Child=@n1.child, Parent=@n1.parent).RelativeTransform.Rotation
   )
   n2 = RigVMDispatchNode(
    Values=((),())
    Index=@RigVMFunction_MathBoolToInteger.Result
    Values.0=@RigVMFunction_MathQuaternionSwingTwist.Swing
    Values.1=@RigVMFunction_MathQuaternionSwingTwist.Twist
   ) {template="DISPATCH_RigVMDispatch_SelectInt32(in Index,in Values,out Result)"}
   Inverse = /Script/RigVM.RigVMFunction_MathQuaternionInverse(Value=@n2.Result) {is=struct RigVMUnitNode, template="Inverse::Execute(in Value,out Result)"}
   > RigVMFunctionReturnNode(outputQuat=RigVMDispatchNode(True=@Inverse.Result, Condition=@n1.invert, False=@n2.Result) {template="DISPATCH_RigVMDispatch_If(in Condition,in True,in False,out Result)"}.Result)
  ]}
 RigVMFunctionLibrary {nodes [
   DeltaQuatExtraction = RigVMCollapseNode(ExecuteContext=(), child=(Type=Bone,Name="None"), parent=(Type=Bone,Name="None"), twistAxis=(X=1,Y=0,Z=0))
   DeltaRotation = RigVMCollapseNode(ExecuteContext=(), Rotation Order=ZYX, premult=1, twistAxis=(X=1,Y=0,Z=0))
   controllers = RigVMCollapseNode(ExecuteContext=(), debug=False)
  ]}
 RigVMFunction_Sequence_1_ContainedGraph {nodes [
   n1 = RigVMFunctionEntryNode()
   RigVMFunction_Sequence = &RigVMFunction_Sequence(ExecuteContext=@n1.ExecuteContext)
   n2 = &RigVMFunction_Sequence(ExecuteContext=@RigVMFunction_Sequence.B)
   RigVMFunctionReturnNode(A=@RigVMFunction_Sequence.A, B=@n2.A, C=@n2.B)
  ]}
 RigVMModel {nodes [
   RigUnit_BeginExecution = &RigUnit_BeginExecution()
   controllers = RigVMFunctionReferenceNode(debug=false, ExecuteContext=@RigUnit_BeginExecution.ExecutePin)
   > RigVMFunction_Sequence = RigVMAggregateNode()
   DeltaQuatExtraction = RigVMFunctionReferenceNode(
    child=(Type=Bone,Name="head")
    onlyTwist=true
    parent=(Type=Null,Name="head_null")
    twistAxis=(X=0,Y=0,Z=1)
    ExecuteContext=@RigVMFunction_Sequence.A
   )
   Interpolate = /Script/RigVM.RigVMFunction_MathQuaternionSlerp(B=(X=0,Y=0,Z=0,W=1), T=0.7, A=@DeltaQuatExtraction.outputQuat) {is=struct RigVMUnitNode, template="Interpolate::Execute(in A,in B,in T,out Result)"}
   > &RigUnit_OffsetTransformForItem(
    Item=(Type=Bone,Name="sleepingBag_head")
    OffsetTransform=(Rotation=(X=0,Y=0,Z=0,W=1),Translation=(X=0,Y=0,Z=0),Scale3D=(X=1,Y=1,Z=1))
    Weight=1
    bPropagateToChildren=True
    OffsetTransform.Rotation=@Interpolate.Result
   )
   &RigUnit_ParentConstraint(
    AdvancedSettings=(RotationOrderForFilter=XZY)
    Child=(Type=Bone,Name="kinEye_l")
    Parents=((Item=(Type=Bone,Name="sleepingBag_head"),Weight=0.2374),(Item=(Type=Bone,Name="face_head"),Weight=0.7626))
    Weight=1
    ExecutePin=@RigVMFunction_Sequence.B
   )
   > &RigUnit_ParentConstraint(
    AdvancedSettings=(RotationOrderForFilter=XZY)
    Child=(Type=Bone,Name="kinEye_r")
    Parents=((Item=(Type=Bone,Name="sleepingBag_head"),Weight=0.2515),(Item=(Type=Bone,Name="face_head"),Weight=0.7485))
    Weight=1
   )
   > &RigUnit_ParentConstraint(
    AdvancedSettings=(RotationOrderForFilter=XZY)
    Child=(Type=Bone,Name="kinEyebrow_l")
    Parents=((Item=(Type=Bone,Name="face_head"),Weight=0.4742),(Item=(Type=Bone,Name="sleepingBag_head"),Weight=0.5258))
    Weight=1
   )
   > &RigUnit_ParentConstraint(
    AdvancedSettings=(RotationOrderForFilter=XZY)
    Child=(Type=Bone,Name="kinEyebrow_r")
    Parents=((Item=(Type=Bone,Name="face_head"),Weight=0.4635),(Item=(Type=Bone,Name="sleepingBag_head"),Weight=0.5365))
    Weight=1
   )
   DeltaRotation = RigVMFunctionReferenceNode(
    Rotation Order=ZYX
    child=head
    debugRotation=false
    max=70
    min=20
    onlyTwist=true
    parent=head_null
    parentType=Null
    premult=1
    primAxis=2
    twistAxis=(X=0,Y=0,Z=1)
    ExecuteContext=@RigVMFunction_Sequence.C
   )
   > &RigUnit_ModifyTransforms(
    ItemToModify=((Item=(Type=Bone,Name="face_head"),Transform=(Rotation=(X=0,Y=-0.0872,Z=0,W=0.9962),Translation=(X=0,Y=0,Z=0),Scale3D=(X=1,Y=1,Z=1))))
    WeightMinimum=0
    Weight=@DeltaRotation.outputWeight
   )
   > DeltaRotation_1 = RigVMFunctionReferenceNode(
    Rotation Order=ZYX
    child=head
    debugRotation=false
    max=70
    min=0
    onlyTwist=true
    parent=head_null
    parentType=Null
    premult=1
    primAxis=2
    twistAxis=(X=0,Y=0,Z=1)
   )
   > &RigUnit_ModifyTransforms(
    ItemToModify=((Item=(Type=Bone,Name="kin_coverHead_ik"),Transform=(Rotation=(X=0,Y=0,Z=0,W=-1),Translation=(X=-5,Y=0,Z=2.75),Scale3D=(X=1,Y=1,Z=1))))
    WeightMinimum=0
    Weight=@DeltaRotation_1.outputWeight
   )
   > &RigUnit_AimConstraintLocalSpaceOffset(
    AdvancedSettings=(DebugSettings=(bEnabled=false,Scale=10,WorldOffset=(Rotation=(X=0,Y=0,Z=0,W=1),Translation=(X=0,Y=0,Z=0),Scale3D=(X=1,Y=1,Z=1))),RotationOrderForFilter=XZY)
    Child=(Type=Bone,Name="kin_cover_belt_l")
    Filter=(bX=True,bY=True,bZ=True)
    Parents=((Item=(Type=Bone,Name="target_cover_belt_l"),Weight=1))
    UpAxis=(X=0,Y=1,Z=0)
    Weight=1
    WorldUp=(Target=(X=0,Y=1,Z=0),Kind=Location,Space=(Type=Null,Name="kin_cover_belt_l_up_null"))
    bMaintainOffset=false
   )
   > &RigUnit_AimConstraintLocalSpaceOffset(
    AdvancedSettings=(DebugSettings=(bEnabled=false,Scale=10,WorldOffset=(Rotation=(X=0,Y=0,Z=0,W=1),Translation=(X=0,Y=0,Z=0),Scale3D=(X=1,Y=1,Z=1))),RotationOrderForFilter=XZY)
    Child=(Type=Bone,Name="kin_cover_belt_r")
    Filter=(bX=True,bY=True,bZ=True)
    Parents=((Item=(Type=Bone,Name="target_cover_belt_r"),Weight=1))
    UpAxis=(X=0,Y=1,Z=0)
    Weight=1
    WorldUp=(Target=(X=0,Y=1,Z=0),Kind=Location,Space=(Type=Null,Name="kin_cover_belt_r_up_null"))
    bMaintainOffset=false
   )
   > &RigUnit_ParentConstraint(
    AdvancedSettings=(InterpolationType=Average,RotationOrderForFilter=XZY)
    Child=(Type=Control,Name="coverHead_pv_ctrl")
    Parents=((Item=(Type=Bone,Name="spine_05"),Weight=1))
    Weight=1
   )
   > /Script/ControlRig.RigUnit_TwoBoneIKSimplePerItem(
    DebugSettings=(bEnabled=False,Scale=10,WorldOffset=(Rotation=(X=0,Y=0,Z=0,W=1),Translation=(X=0,Y=0,Z=0),Scale3D=(X=1,Y=1,Z=1)))
    EffectorItem=(Type=Bone,Name="kin_coverHead_3")
    ItemA=(Type=Bone,Name="kin_coverHead_1")
    ItemALength=0
    ItemB=(Type=Bone,Name="kin_coverHead_2")
    ItemBLength=0
    PoleVectorKind=Location
    PoleVectorSpace=(Type=Bone,Name="None")
    PrimaryAxis=(X=1,Y=0,Z=0)
    SecondaryAxis=(X=0,Y=0,Z=1)
    SecondaryAxisWeight=1
    StretchMaximumRatio=2
    StretchStartRatio=1
    Weight=1
    bEnableStretch=true
    bPropagateToChildren=True
    PoleVector=&RigUnit_GetTransform(Item=(Type=Control,Name="coverHead_pv_ctrl"), bInitial=False).Transform.Translation
    Effector=&RigUnit_GetTransform(Item=(Type=Bone,Name="kin_coverHead_ik"), bInitial=False).Transform
   ) {is=struct RigVMUnitNode}
  ]}
 RigVMModel Rig {nodes [
   n1 = &RigUnit_PrepareForExecution()
   ComputePoleVector = RigVMFunctionReferenceNode(
    Bone A=(Type=Bone,Name="kin_coverHead_1")
    Bone B=(Type=Bone,Name="kin_coverHead_2")
    Bone C=(Type=Bone,Name="kin_coverHead_3")
    Draw Transform=true
    Initial=true
    OffsetFactor=10
    Origin Scale=(X=0,Y=0,Z=0)
    ExecuteContext=@n1.ExecutePin
   )
   > &RigUnit_SetTransform(
    Item=(Type=Control,Name="coverHead_pv_ctrl")
    Value=(Rotation=(X=0,Y=0,Z=0,W=1),Translation=(X=0,Y=0,Z=0),Scale3D=(X=1,Y=1,Z=1))
    Weight=1
    bInitial=true
    Value.Translation=@ComputePoleVector.Transform.Translation
   )
  ]}
]
hierarchy [
 root {loc=<0 0 0>, rot=<0 0 0 1>}
 pelvis {parent=root, loc=<0 -0.7321 92.4218>, rot=<0.0127 0.707 -0.0127 -0.707>}
 spine_01 {parent=pelvis, loc=<5.6002 -0.5128 0>, rot=<0 0 0.034 -0.9994>}
 spine_02 {parent=spine_01, loc=<8.0756 0.0139 0>, rot=<0 0 0.0238 -0.9997>}
 spine_03 {parent=spine_02, loc=<7.0395 0 0>, rot=<0 0 0.0149 -0.9999>}
 spine_04 {parent=spine_03, loc=<8.1023 0 0>, rot=<0 0 -0.0132 -0.9999>}
 spine_05 {parent=spine_04, loc=<10.0141 0 0>, rot=<0 0 -0.0015 -1>}
 neck_01 {parent=spine_05, loc=<10.815 1.9196 0>, rot=<0 0 -0.1864 -0.9825>}
 neck_02 {parent=neck_01, loc=<4.7347 0.3045 0>, rot=<0 0 0.0363 -0.9993>}
 head {parent=neck_02, loc=<5.7632 0.5552 0>, rot=<0 0 0.115 -0.9934>}
 faceAttach {parent=head, loc=<-4.146 -0.8545 0>, rot=<0.0029 -0.7071 -0.0029 -0.7071>}
 C_brow_mid {parent=faceAttach, loc=<-0.1035 10.4957 10.4678>, rot=<0 -0.0822 0.0162 -0.9965>}
 L_brow_mid {parent=faceAttach, loc=<4.1797 9.635 11.5677>, rot=<-0.0013 -0.0132 0.2553 -0.9668>}
 L_brow_outer {parent=faceAttach, loc=<7.1106 6.7147 11.8525>, rot=<0.1225 0.1503 0.4052 -0.8934>}
 L_cheek_inner {parent=faceAttach, loc=<5.0718 8.2272 4.4314>, rot=<-0.0141 -0.1263 0.3343 -0.9339>}
 C_lip_upper_mid {parent=faceAttach, loc=<-0.0729 9.4441 2.5422>, rot=<0.2225 0.0015 -0.0006 -0.9749>}
 L_lip_upper_outer {parent=faceAttach, loc=<1.5605 8.8845 2.7078>, rot=<0.0449 -0.0474 0.2531 -0.9652>}
 L_lip_corner {parent=faceAttach, loc=<3.0705 9.1224 2.2497>, rot=<0.0021 0.0089 -0.0145 -0.9999>}
 C_nose_bridge {parent=faceAttach, loc=<-0.141 11.9139 5.6533>, rot=<-0.2218 0.0038 0.0011 -0.9751>}
 R_brow_mid {parent=faceAttach, loc=<-4.3215 9.6411 11.5677>, rot=<-0.0146 0.1156 -0.2545 -0.96>}
 R_brow_outer {parent=faceAttach, loc=<-7.2593 6.7377 11.8319>, rot=<0.0752 -0.0908 -0.4219 -0.899>}
 R_cheek_inner {parent=faceAttach, loc=<-5.2061 8.2415 4.4065>, rot=<-0.0127 0.1226 -0.3346 -0.9343>}
 R_lip_corner {parent=faceAttach, loc=<-3.0396 9.2358 2.326>, rot=<0.0027 -0.0082 0.0146 -0.9999>}
 C_jaw {parent=faceAttach, loc=<-0.1507 2.7303 1.817>, rot=<0 0 0 1>}
 L_lip_lower_outer {parent=C_jaw, loc=<1.7512 6.1661 0.4653>, rot=<0.0109 0.033 0.2969 -0.9543>}
 C_lip_lower_mid {parent=C_jaw, loc=<0.0631 6.7698 0.7477>, rot=<0 -0.0397 -0.0084 -0.9992>}
 R_lip_lower_outer {parent=C_jaw, loc=<-1.7726 6.5553 0.5214>, rot=<-0.0012 0.0321 -0.2464 -0.9686>}
 teeth_lower {parent=C_jaw, loc=<0.0101 2.9913 -1.6093>, rot=<0 0 0 1>}
 tongue {parent=C_jaw, loc=<0.1002 1.6559 -1.215>, rot=<-0.1568 0 0 -0.9876>}
 R_eye {parent=faceAttach, loc=<-4.346 8.7158 9.1288>, rot=<0 0 0.9998 -0.0187>}
 R_eye_lid_upper_mid {parent=R_eye, loc=<-0.0396 -1.1714 0.2631>, rot=<0.0523 -0.0009 0.0167 -0.9985>}
 R_eye_lid_lower_mid {parent=R_eye, loc=<-0.0085 -0.9425 -0.7051>, rot=<0.0329 -0.0005 0.0167 -0.9993>}
 dyn_R_eye {parent=R_eye, loc=<0 0 0>, rot=<0.0074 -0.0001 0.0187 -0.9998>, scale=<1.1236 1.1236 1.1236>}
 L_eye {parent=faceAttach, loc=<4.1921 8.7158 9.1288>, rot=<0 0 0.9998 0.0187>}
 L_eye_lid_upper_mid {parent=L_eye, loc=<0.0251 -1.1623 0.2388>, rot=<0.0523 0 0 -0.9986>}
 L_eye_lid_lower_mid {parent=L_eye, loc=<0.0223 -0.923 -0.7488>, rot=<0.0329 0 0 -0.9995>}
 dyn_L_eye {parent=L_eye, loc=<0 0 0>, rot=<0.0074 0.0001 -0.0187 -0.9998>, scale=<1.1236 1.1236 1.1236>}
 teeth_upper {parent=faceAttach, loc=<-0.088 6.9333 2.9004>, rot=<0 0 0 1>}
 R_lip_upper_outer {parent=faceAttach, loc=<-1.7121 8.9404 2.7423>, rot=<0.0461 0.0441 -0.2518 -0.9657>}
 face_01_r {parent=faceAttach, loc=<-8.6973 4.0157 3.5609>, rot=<0.3123 -0.2839 0.781 -0.4604>, scale=<1.1236 1.1236 1.1236>}
 face_04_l {parent=faceAttach, loc=<6.1991 1.8671 -2.0013>, rot=<-0.556 -0.5464 0.5864 0.2203>, scale=<1.1236 1.1236 1.1236>}
 face_01_l {parent=faceAttach, loc=<8.6973 4.0157 3.5609>, rot=<-0.4604 -0.781 -0.2839 -0.3123>, scale=<1.1236 1.1236 1.1236>}
 dyn_hair_01 {parent=head, loc=<13.5645 3.6893 0>, rot=<0.0082 -0.7071 -0.0082 -0.7071>}
 dyn_hair_01_a {parent=dyn_hair_01, loc=<1.1549 -1.7351 3.7502>, rot=<-0.3421 0.048 -0.1686 -0.9231>}
 dyn_hair_01_b {parent=dyn_hair_01, loc=<1.4327 -2.2816 1.9183>, rot=<-0.4371 0.0219 0.178 -0.8813>}
 dyn_hair_01_c {parent=dyn_hair_01, loc=<-5.8268 2.0739 1.5328>, rot=<0 0 0 1>}
 dyn_hair_01_d {parent=dyn_hair_01, loc=<-0.3149 -7.4732 1.5573>, rot=<0 0 0 1>}
 dyn_hair_01_e {parent=dyn_hair_01, loc=<6.3939 -2.3971 -0.0355>, rot=<0 0 0 1>}
 dyn_hair_01_f {parent=dyn_hair_01, loc=<-4.004 8.6437 -1.016>, rot=<0 0 0 1>}
 dyn_hair_01_g {parent=dyn_hair_01, loc=<0.3972 11.224 -1.0613>, rot=<0 0 0 1>}
 dyn_hair_01_h {parent=dyn_hair_01, loc=<6.6634 -10.5445 -5.47>, rot=<0 0 0 1>}
 dyn_hair_01_i {parent=dyn_hair_01, loc=<-7.4034 -2.5332 5.4989>, rot=<0 0 0 1>}
 dyn_hair_02 {parent=head, loc=<12.2503 -5.2722 0>, rot=<0.0082 -0.7071 -0.0082 -0.7071>}
 dyn_hair_02_a {parent=dyn_hair_02, loc=<0.1406 -1.6929 4.5073>, rot=<0 0 0 1>}
 dyn_hair_02_b {parent=dyn_hair_02, loc=<-0.8743 -5.6147 0.9456>, rot=<0 0 0 1>}
 dyn_hair_02_c {parent=dyn_hair_02, loc=<1.9636 -5.2659 0.4978>, rot=<0 0.0111 0 -0.9999>}
 dyn_hair_02_d {parent=dyn_hair_02, loc=<-2.4971 -4.258 1.5798>, rot=<0 0 0 1>}
 dyn_hair_03 {parent=head, loc=<12.9532 2.328 -4.2253>, rot=<-0.1663 -0.6873 -0.1821 -0.6833>}
 dyn_hair_03_a {parent=dyn_hair_03, loc=<-7.2709 -5.2263 0.2429>, rot=<0 0 0 1>}
 dyn_hair_03_b {parent=dyn_hair_03, loc=<0 0 0>, rot=<0 0 0 1>}
 dyn_hair_04 {parent=head, loc=<14.3902 6.6981 -1.9471>, rot=<0.1534 -0.6903 0.1375 -0.6936>}
 dyn_hair_04_a {parent=dyn_hair_04, loc=<0.6961 0.4973 -3.7811>, rot=<0.1905 -0.5944 -0.2385 -0.744>}
 dyn_hair_05 {parent=head, loc=<12.736 6.2945 2.5401>, rot=<0.0082 -0.7071 -0.0082 -0.7071>}
 dyn_hair_05_a {parent=dyn_hair_05, loc=<2.676 1.3272 -4.3488>, rot=<0.1063 -0.4535 -0.2019 -0.8616>}
 dyn_hair_05_b {parent=dyn_hair_05_a, loc=<5.8133 0.6663 -0.8464>, rot=<-0.0053 -0.1632 -0.117 -0.9796>}
 dyn_hair_05_c {parent=dyn_hair_05_b, loc=<4.2864 -0.389 -2.1082>, rot=<-0.0904 0.1037 0.5045 -0.8524>}
 dyn_hair_06 {parent=head, loc=<11.805 4.1522 -4.9679>, rot=<0.0082 -0.7071 -0.0082 -0.7071>}
 dyn_hair_06_a {parent=dyn_hair_06, loc=<-4.1153 1.4743 -2.0266>, rot=<0 0 0 1>}
 dyn_hair_06_b {parent=dyn_hair_06_a, loc=<5.1935 0 0>, rot=<0 0 -0.4584 -0.8887>}
 dyn_hair_06_c {parent=dyn_hair_06_b, loc=<6.5393 0 0>, rot=<0 0 0.4122 -0.9111>}
 dyn_hair_07 {parent=head, loc=<13.2415 2.8247 -3.6848>, rot=<0.0082 -0.7071 -0.0082 -0.7071>}
 dyn_hair_07_a {parent=dyn_hair_07, loc=<5.0613 -0.2989 -0.1927>, rot=<-0.0408 0.1307 -0.0818 -0.9872>}
 dyn_hair_07_b {parent=dyn_hair_07_a, loc=<9.4763 0 0>, rot=<0 0 0 1>}
 dyn_hair_08 {parent=head, loc=<5.9542 -7.9679 0.012>, rot=<0.0082 -0.7071 -0.0082 -0.7071>}
 dyn_hair_08_a {parent=dyn_hair_08, loc=<-5.0611 0.2989 0.1925>, rot=<-0.0408 0.1307 -0.0818 -0.9872>}
 dyn_hair_08_c {parent=dyn_hair_08_a, loc=<-10.2943 -0.0087 0.4996>, rot=<0 0 0 1>}
 dyn_hair_09 {parent=head, loc=<3.2466 -5.9685 0.0247>, rot=<0.0082 -0.7071 -0.0082 -0.7071>}
 dyn_earring_r {parent=head, loc=<-0.3062 -0.0183 6.0052>, rot=<0.6791 0.2873 -0.6271 -0.2511>}
 dyn_earring_l {parent=head, loc=<-0.4315 -0.0846 -5.856>, rot=<0.257 -0.654 0.285 -0.6519>}
 dyn_hair_011 {parent=head, loc=<13.3938 0.095 8.72>, rot=<0.0029 -0.7071 -0.0029 -0.7071>}
 dyn_hair_010 {parent=head, loc=<2.9726 -5.2891 7.151>, rot=<0.0029 -0.7071 -0.0029 -0.7071>}
 dyn_hair_012 {parent=head, loc=<10.9514 6.9867 -4.7288>, rot=<0.0728 -0.9299 0.2743 -0.2341>}
 dyn_hair_013 {parent=head, loc=<11.3311 7.7673 -2.0361>, rot=<-0.2054 -0.7042 -0.2316 -0.6389>}
 dyn_hair_013_a {parent=dyn_hair_013, loc=<2.7341 0 0>, rot=<0 0 0 1>}
 dyn_hair_014 {parent=head, loc=<11.6678 7.6279 3.7388>, rot=<0.6407 0.0954 0.7581 0.0754>}
 dyn_hair_014_a {parent=dyn_hair_014, loc=<3.2674 0 0>, rot=<0 0 0 1>}
 dyn_hair_015 {parent=head, loc=<10.6707 8.0888 1.3104>, rot=<0.6542 0.2188 0.6956 0.2009>}
 dyn_hair_015_a {parent=dyn_hair_015, loc=<3.0275 0 0>, rot=<0 0 0 1>}
 Tail {parent=head, loc=<4.0061 -2.4566 0>, rot=<0.0029 -0.7071 -0.0029 -0.7071>}
 dyn_tail_01 {parent=Tail, loc=<0 -1.8481 0>, rot=<0 0 0 1>}
 fin_l {parent=head, loc=<4.2096 4.0295 -7.0475>, rot=<0.6969 -0.1199 -0.6959 0.1257>}
 dyn_fin_l {parent=fin_l, loc=<2.9083 -0.5249 -0.0123>, rot=<0 0 0.0218 -0.9998>}
 fin_r {parent=head, loc=<4.2091 4.0295 7.0475>, rot=<-0.1199 -0.6969 -0.1257 -0.6959>}
 dyn_fin_r {parent=fin_r, loc=<-2.8876 0.4538 0>, rot=<0 0 0 1>}
 dyn_earing_r {parent=head, loc=<0.5805 -0.2012 7.078>, rot=<1 0.0041 0 0>}
 dyn_earing_l {parent=head, loc=<0.5802 -0.2012 -7.078>, rot=<0.0041 -1 0 0>}
 dyn_hairBase {parent=head, loc=<10.4256 7.1993 2.9441>, rot=<-0.3373 0.0989 -0.1767 -0.9194>}
 dyn_hairMid {parent=dyn_hairBase, loc=<6.1134 0 0>, rot=<0.1398 -0.1326 0.3037 -0.9331>}
 hair_back_right {parent=head, loc=<-0.6973 -6.8131 4.1022>, rot=<0.0029 -0.7071 -0.0029 -0.7071>}
 hair_back_left {parent=head, loc=<-0.6409 -6.1476 -4.5356>, rot=<0.0029 -0.7071 -0.0029 -0.7071>}
 hair_side_left {parent=head, loc=<9.8305 -0.9288 -6.0046>, rot=<0.0029 -0.7071 -0.0029 -0.7071>}
 hair_side_right {parent=head, loc=<11.3604 -0.8243 5.3489>, rot=<0.0029 -0.7071 -0.0029 -0.7071>}
 hair_front {parent=head, loc=<10.5965 10.0795 0.1371>, rot=<0.0029 -0.7071 -0.0029 -0.7071>}
 dyn_ear_l_master {parent=head, loc=<17.0915 -3.5506 -5.8443>, rot=<0 0 0 1>}
 dyn_ear01_l {parent=dyn_ear_l_master, loc=<0 0 0>, rot=<0.0224 -0.1578 0.1304 -0.9786>}
 dyn_ear02_l {parent=dyn_ear01_l, loc=<3.5749 0 0>, rot=<-0.0017 0.0088 -0.0846 -0.9964>}
 dyn_ear03_l {parent=dyn_ear02_l, loc=<6.419 0 0>, rot=<0.0033 0.0792 -0.1438 -0.9864>}
 dyn_ear04_l {parent=dyn_ear03_l, loc=<4.2719 0 0>, rot=<0.0192 0.0546 -0.1522 -0.9867>}
 dyn_ear05_l {parent=dyn_ear04_l, loc=<4.3942 0 0>, rot=<0 0 0 1>}
 dyn_ear_r_master {parent=head, loc=<17.0923 -3.5619 5.8346>, rot=<0 0 0 1>}
 dyn_ear01_r {parent=dyn_ear_r_master, loc=<0 0 0>, rot=<-0.0223 0.1593 0.129 -0.9785>}
 dyn_ear02_r {parent=dyn_ear01_r, loc=<3.5743 0 0>, rot=<0.0019 -0.0103 -0.0841 -0.9964>}
 dyn_ear03_r {parent=dyn_ear02_r, loc=<6.4514 0 0>, rot=<-0.0033 -0.0783 -0.1434 -0.9865>}
 dyn_ear04_r {parent=dyn_ear03_r, loc=<4.2229 0 0>, rot=<-0.0184 -0.056 -0.1299 -0.9898>}
 dyn_ear05_r {parent=dyn_ear04_r, loc=<3.0399 0 0>, rot=<0 0 0 1>}
 dyn_ear_l_MSTR {parent=head, loc=<11.9817 2.4267 -6.5386>, rot=<0.8593 0.2015 -0.0197 -0.4697>}
 dyn_ear_l_01 {parent=dyn_ear_l_MSTR, loc=<0 0 0>, rot=<0 0 0 1>}
 dyn_ear_l_02 {parent=dyn_ear_l_01, loc=<2.7674 0 0>, rot=<0.0014 -0.0527 -0.0251 -0.9983>}
 dyn_ear_l_03 {parent=dyn_ear_l_02, loc=<2.7674 0 0>, rot=<0 0 0 1>}
 dyn_ear_r_MSTR {parent=head, loc=<11.9821 2.4267 6.5386>, rot=<0.2015 -0.8593 0.4697 -0.0197>}
 dyn_ear_r_01 {parent=dyn_ear_r_MSTR, loc=<0 0 0>, rot=<0 0 0 1>}
 dyn_ear_r_02 {parent=dyn_ear_r_01, loc=<-2.7666 0.0003 -0.0002>, rot=<0.0014 -0.0527 -0.0251 -0.9983>}
 dyn_ear_r_03 {parent=dyn_ear_r_02, loc=<-2.7681 -0.0002 0.0001>, rot=<0 0 0 1>}
 FACIAL_C_FacialRoot {parent=head, loc=<2.1524 -0.746 -0.0002>, rot=<0.4879 -0.5118 -0.4879 -0.5118>}
 FACIAL_C_Forehead {parent=FACIAL_C_FacialRoot, loc=<0.0003 -2.7714 8.7804>, rot=<-0.0499 0 0 -0.9988>}
 FACIAL_L_ForeheadIn {parent=FACIAL_C_FacialRoot, loc=<1.6489 -2.6487 8.7066>, rot=<-0.0672 -0.0696 0.0038 -0.9953>}
 FACIAL_R_ForeheadIn {parent=FACIAL_C_FacialRoot, loc=<-1.6481 -2.6466 8.707>, rot=<-0.0672 0.0696 -0.0038 -0.9953>}
 FACIAL_L_ForeheadMid {parent=FACIAL_C_FacialRoot, loc=<3.0937 -2.8874 8.2585>, rot=<-0.0774 -0.1755 -0.0217 -0.9812>}
 FACIAL_R_ForeheadMid {parent=FACIAL_C_FacialRoot, loc=<-3.0853 -2.8863 8.2578>, rot=<-0.0774 0.1755 0.0218 -0.9812>}
 FACIAL_L_ForeheadOut {parent=FACIAL_C_FacialRoot, loc=<4.4146 -2.8714 7.4528>, rot=<-0.1276 -0.2551 -0.0808 -0.9551>}
 FACIAL_R_ForeheadOut {parent=FACIAL_C_FacialRoot, loc=<-4.4049 -2.8724 7.4484>, rot=<-0.1276 0.2551 0.0808 -0.9551>}
 FACIAL_L_EyesackUpper {parent=FACIAL_C_FacialRoot, loc=<3.0818 -1.6784 8.7377>, rot=<0.0423 -0.0838 -0.0651 -0.9935>}
 FACIAL_R_EyesackUpper {parent=FACIAL_C_FacialRoot, loc=<-3.0815 -1.6731 8.7365>, rot=<0.0423 0.0838 0.0651 -0.9935>}
 FACIAL_L_EyelidUpperFurrow {parent=FACIAL_C_FacialRoot, loc=<2.9342 -1.4364 8.64>, rot=<-0.0365 -0.1503 -0.0806 -0.9847>}
 FACIAL_R_EyelidUpperFurrow {parent=FACIAL_C_FacialRoot, loc=<-2.9313 -1.4274 8.6385>, rot=<-0.0365 0.1503 0.0806 -0.9847>}
 FACIAL_L_EyelidUpperA {parent=FACIAL_C_FacialRoot, loc=<2.9763 -0.7851 6.9344>, rot=<-0.1543 0 0 -0.988>}
 FACIAL_R_EyelidUpperA {parent=FACIAL_C_FacialRoot, loc=<-2.9766 -0.7828 6.9396>, rot=<-0.1543 0 0 -0.988>}
 FACIAL_L_Eye {parent=FACIAL_C_FacialRoot, loc=<2.9834 -0.7677 6.9671>, rot=<-0.0065 0 0 -1>}
 FACIAL_R_Eye {parent=FACIAL_C_FacialRoot, loc=<-2.9852 -0.7659 6.9693>, rot=<-0.0065 0 0 -1>}
 FACIAL_L_EyelidLowerA {parent=FACIAL_C_FacialRoot, loc=<2.9763 -0.7851 6.9344>, rot=<0.2778 0 0 -0.9607>}
 FACIAL_R_EyelidLowerA {parent=FACIAL_C_FacialRoot, loc=<-2.9766 -0.7828 6.9396>, rot=<0.2777 0 0 -0.9607>}
 FACIAL_L_EyeCornerInner {parent=FACIAL_C_FacialRoot, loc=<1.8971 -0.2481 8.01>, rot=<-0.0065 0 0 -1>}
 FACIAL_R_EyeCornerInner {parent=FACIAL_C_FacialRoot, loc=<-1.8703 -0.2298 8.0119>, rot=<-0.0065 0 0 -1>}
 FACIAL_L_EyeCornerOuter {parent=FACIAL_C_FacialRoot, loc=<4.5915 -0.8927 7.4384>, rot=<-0.006 -0.3826 -0.0025 -0.9239>}
 FACIAL_R_EyeCornerOuter {parent=FACIAL_C_FacialRoot, loc=<-4.6085 -0.8951 7.4005>, rot=<-0.006 0.3826 0.0025 -0.9239>}
 FACIAL_L_EyesackLower {parent=FACIAL_C_FacialRoot, loc=<3.2288 0.9149 7.1307>, rot=<-0.0287 -0.1396 -0.0232 -0.9895>}
 FACIAL_R_EyesackLower {parent=FACIAL_C_FacialRoot, loc=<-3.1784 0.8923 7.1642>, rot=<-0.0287 0.1396 0.0232 -0.9895>}
 FACIAL_L_CheekInner {parent=FACIAL_C_FacialRoot, loc=<2.0463 1.4617 7.7395>, rot=<-0.1295 -0.0867 -0.1977 -0.9678>}
 FACIAL_R_CheekInner {parent=FACIAL_C_FacialRoot, loc=<-2.0395 1.4606 7.7469>, rot=<-0.1295 0.0867 0.1977 -0.9678>}
 FACIAL_L_CheekOuter {parent=FACIAL_C_FacialRoot, loc=<3.9119 1.52 6.2643>, rot=<0.0492 -0.3776 0.0979 -0.9195>}
 FACIAL_R_CheekOuter {parent=FACIAL_C_FacialRoot, loc=<-3.8938 1.5218 6.2686>, rot=<0.0492 0.3776 -0.0979 -0.9195>}
 FACIAL_L_NasolabialBulge {parent=FACIAL_C_FacialRoot, loc=<3.1013 4.8026 6.2299>, rot=<0.0519 -0.3775 0.0214 -0.9243>}
 FACIAL_R_NasolabialBulge {parent=FACIAL_C_FacialRoot, loc=<-3.0992 4.8017 6.2318>, rot=<0.0519 0.3775 -0.0213 -0.9243>}
 FACIAL_L_NasolabialFurrow {parent=FACIAL_C_FacialRoot, loc=<2.787 3.5798 8.4843>, rot=<-0.0393 -0.1235 -0.3644 -0.9222>}
 FACIAL_R_NasolabialFurrow {parent=FACIAL_C_FacialRoot, loc=<-2.7857 3.58 8.4849>, rot=<-0.0393 0.1234 0.3645 -0.9222>}
 FACIAL_L_CheekLower {parent=FACIAL_C_FacialRoot, loc=<3.4202 4.6229 5.596>, rot=<0.0287 -0.4273 0.055 -0.902>}
 FACIAL_R_CheekLower {parent=FACIAL_C_FacialRoot, loc=<-3.4174 4.6222 5.5981>, rot=<0.0287 0.4272 -0.0549 -0.902>}
 FACIAL_C_Nose {parent=FACIAL_C_FacialRoot, loc=<-0.0002 0.8449 9.0042>, rot=<-0.1368 0 0 -0.9906>}
 FACIAL_C_NoseLower {parent=FACIAL_C_Nose, loc=<-0.0001 2.6719 -0.3289>, rot=<0.1305 0 0 -0.9914>}
 FACIAL_L_Nostril {parent=FACIAL_C_Nose, loc=<0.6797 2.0796 -0.2034>, rot=<0.0173 0.1305 0.0023 -0.9913>}
 FACIAL_R_Nostril {parent=FACIAL_C_Nose, loc=<-0.6778 2.0746 -0.2086>, rot=<0.0173 -0.1305 -0.0023 -0.9913>}
 FACIAL_C_NoseTip {parent=FACIAL_C_Nose, loc=<-0.0002 2.1349 0.4306>, rot=<0.0872 0 0 -0.9962>}
 FACIAL_C_MouthUpper {parent=FACIAL_C_FacialRoot, loc=<0.0007 4.9989 7.1849>, rot=<-0.0065 0 0 -1>}
 FACIAL_C_LipUpper {parent=FACIAL_C_MouthUpper, loc=<-0.0004 -0.0035 1.7864>, rot=<-0.0872 0 0 -0.9962>}
 FACIAL_L_LipUpper {parent=FACIAL_C_MouthUpper, loc=<0.8927 0.0584 1.7116>, rot=<-0.0778 -0.1601 -0.0379 -0.9833>}
 FACIAL_R_LipUpper {parent=FACIAL_C_MouthUpper, loc=<-0.8949 0.0577 1.712>, rot=<-0.0778 0.1601 0.0379 -0.9833>}
 FACIAL_L_LipUpperOuter {parent=FACIAL_C_MouthUpper, loc=<1.6456 0.2182 1.302>, rot=<-0.0984 -0.2293 -0.0983 -0.9634>}
 FACIAL_R_LipUpperOuter {parent=FACIAL_C_MouthUpper, loc=<-1.6469 0.219 1.3023>, rot=<-0.0984 0.2293 0.0983 -0.9634>}
 FACIAL_L_LipCorner {parent=FACIAL_C_MouthUpper, loc=<2.132 0.5273 0.5574>, rot=<0 -0.2078 0 -0.9782>}
 FACIAL_R_LipCorner {parent=FACIAL_C_MouthUpper, loc=<-2.1325 0.527 0.5573>, rot=<0 0.2078 0 -0.9782>}
 FACIAL_L_Masseter {parent=FACIAL_C_FacialRoot, loc=<6.0998 2.0963 4.1921>, rot=<0.0629 0.1128 0.0086 -0.9916>}
 FACIAL_R_Masseter {parent=FACIAL_C_FacialRoot, loc=<-6.0991 2.0965 4.1909>, rot=<0.0629 -0.1128 -0.0086 -0.9916>}
 FACIAL_C_UnderChin {parent=FACIAL_C_FacialRoot, loc=<-0.0001 8.5699 3.8497>, rot=<0.4942 0 0 -0.8693>}
 FACIAL_L_UnderChin {parent=FACIAL_C_FacialRoot, loc=<2.68 7.4568 3.0363>, rot=<0.3971 -0.3554 0.08 -0.8424>}
 FACIAL_R_UnderChin {parent=FACIAL_C_FacialRoot, loc=<-2.6793 7.4561 3.0377>, rot=<0.3971 0.3554 -0.08 -0.8424>}
 FACIAL_C_TeethUpper {parent=FACIAL_C_FacialRoot, loc=<-0.0002 5.6892 8.1383>, rot=<-0.0065 0 0 -1>}
 FACIAL_C_LowerLipRotation {parent=FACIAL_C_FacialRoot, loc=<-0.0006 2.6342 2.1678>, rot=<0.3359 0 0 -0.9419>}
 FACIAL_C_MouthLower {parent=FACIAL_C_LowerLipRotation, loc=<0.0012 -1.3438 5.3811>, rot=<-0.342 0 0 -0.9397>}
 FACIAL_C_LipLower {parent=FACIAL_C_MouthLower, loc=<-0.0002 1.2067 1.4822>, rot=<0.0872 0 0 -0.9962>}
 FACIAL_L_LipLower {parent=FACIAL_C_MouthLower, loc=<0.7997 1.1825 1.3646>, rot=<0.0927 -0.1294 0.0272 -0.9869>}
 FACIAL_R_LipLower {parent=FACIAL_C_MouthLower, loc=<-0.8016 1.1833 1.3643>, rot=<0.0927 0.1294 -0.0272 -0.9869>}
 FACIAL_L_LipLowerOuter {parent=FACIAL_C_MouthLower, loc=<1.5068 0.939 1.0182>, rot=<0.0913 -0.2321 0.0318 -0.9679>}
 FACIAL_R_LipLowerOuter {parent=FACIAL_C_MouthLower, loc=<-1.51 0.9394 1.0188>, rot=<0.0913 0.2321 -0.0318 -0.9679>}
 FACIAL_C_TeethLower {parent=FACIAL_C_LowerLipRotation, loc=<0.0001 -1.3728 6.4905>, rot=<-0.342 0 0 -0.9397>}
 FACIAL_C_Tongue1 {parent=FACIAL_C_TeethLower, loc=<0.0034 -0.1082 -4.6559>, rot=<-0.2588 0 0 -0.9659>}
 FACIAL_C_Tongue2 {parent=FACIAL_C_Tongue1, loc=<-0.001 0.4182 1.3484>, rot=<0.2588 0 0 -0.9659>}
 FACIAL_C_Tongue3 {parent=FACIAL_C_Tongue2, loc=<-0.001 0.2347 1.2375>, rot=<0.1305 0 0 -0.9914>}
 FACIAL_C_Tongue4 {parent=FACIAL_C_Tongue3, loc=<-0.0001 -0.1198 1.0773>, rot=<0.0436 0 0 -0.999>}
 FACIAL_C_TongueUpper3 {parent=FACIAL_C_Tongue3, loc=<0.0006 -0.3247 -0.1806>, rot=<0 0 0 1>}
 FACIAL_C_TongueLower3 {parent=FACIAL_C_Tongue3, loc=<-0.0003 0.7971 0.0192>, rot=<0 0 0 1>}
 FACIAL_L_TongueSide3 {parent=FACIAL_C_Tongue3, loc=<1.4683 0.0026 0.1167>, rot=<0 0 0 1>}
 FACIAL_R_TongueSide3 {parent=FACIAL_C_Tongue3, loc=<-1.4695 0.0021 0.1176>, rot=<0 0 0 1>}
 FACIAL_C_TongueUpper2 {parent=FACIAL_C_Tongue2, loc=<0.0011 -0.5596 -0.2075>, rot=<0 0 0 1>}
 FACIAL_L_TongueSide2 {parent=FACIAL_C_Tongue2, loc=<1.7378 0.0204 0.1647>, rot=<0 0 0 1>}
 FACIAL_R_TongueSide2 {parent=FACIAL_C_Tongue2, loc=<-1.74 0.0202 0.1663>, rot=<0 0 0 1>}
 FACIAL_C_Jaw {parent=FACIAL_C_FacialRoot, loc=<-0.0006 2.6342 2.1678>, rot=<0.3359 0 0 -0.9419>}
 FACIAL_C_Jawline {parent=FACIAL_C_Jaw, loc=<0.0005 2.4182 7.0731>, rot=<0.2586 0 0 -0.966>}
 FACIAL_L_ChinSide {parent=FACIAL_C_Jaw, loc=<2.4587 0.2989 6.8168>, rot=<-0.2208 -0.3738 0.0499 -0.8995>}
 FACIAL_R_ChinSide {parent=FACIAL_C_Jaw, loc=<-2.4562 0.2994 6.8156>, rot=<-0.2208 0.3738 -0.0499 -0.8995>}
 FACIAL_C_Chin {parent=FACIAL_C_Jaw, loc=<0.0005 1.0237 7.3158>, rot=<-0.2591 0 0 -0.9659>}
 FACIAL_L_Ear {parent=FACIAL_C_FacialRoot, loc=<6.3127 0.2036 0.5771>, rot=<-0.0328 -0.0005 -0.0261 -0.9991>}
 FACIAL_R_Ear {parent=FACIAL_C_FacialRoot, loc=<-6.3126 0.2039 0.5745>, rot=<-0.0328 0.0005 0.0262 -0.9991>}
 FACIAL_L_Jawline {parent=FACIAL_C_FacialRoot, loc=<4.1189 5.7978 4.6505>, rot=<0.203 0.1581 -0.0536 -0.9648>}
 FACIAL_R_Jawline {parent=FACIAL_C_FacialRoot, loc=<-4.1197 5.7992 4.6498>, rot=<0.203 -0.1581 0.0536 -0.9648>}
 FACIAL_C_NoseBridge {parent=FACIAL_C_FacialRoot, loc=<0.0002 -1.2439 9.3817>, rot=<0.0373 0 0 -0.9993>}
 FACIAL_L_EyelidUpperB {parent=FACIAL_C_FacialRoot, loc=<2.9763 -0.7851 6.9344>, rot=<-0.2313 0 0 -0.9729>}
 FACIAL_L_EyelidLowerB {parent=FACIAL_C_FacialRoot, loc=<2.9763 -0.7851 6.9344>, rot=<0.3111 0 0 -0.9504>}
 FACIAL_R_EyelidLowerB {parent=FACIAL_C_FacialRoot, loc=<-2.9766 -0.7828 6.9396>, rot=<0.3194 0 0 -0.9476>}
 FACIAL_R_EyelidUpperB {parent=FACIAL_C_FacialRoot, loc=<-2.9766 -0.7828 6.9396>, rot=<-0.2313 0 0 -0.9729>}
 FX_EyeTrail_L {parent=head, type=Null, loc=<4.75 8 2.6>, rot=<-0.0051 0.6948 0.0011 0.7192>}
 FX_EyeTrail_R {parent=head, type=Null, loc=<4.75 8 -2.831>, rot=<0 0.7071 0 0.7071>}
 FX_Eye_L {parent=head, type=Null, loc=<3.4814 5.9142 -2.8258>, rot=<-0.0059 0.7071 0 0.7071>}
 FX_Eye_R {parent=head, type=Null, loc=<3.4814 5.9142 2.8258>, rot=<-0.0059 0.7071 0 0.7071>}
 earpiece {parent=head, type=Null, loc=<0 0 0>, rot=<0 0 0 1>}
 hat {parent=head, type=Null, loc=<0 0 0>, rot=<0 0.7071 0 0.7071>}
 dyn_neck_01 {parent=neck_01, loc=<-0.6307 -10.4769 5.0022>, rot=<-0.1086 -0.6987 0.1086 -0.6987>}
 dyn_neck_02 {parent=neck_01, loc=<-0.6116 -9.0665 -6.1366>, rot=<-0.1086 -0.6987 0.1086 -0.6987>}
 dyn_neck_03 {parent=neck_01, loc=<4.3124 0.1997 6.9457>, rot=<-0.1086 -0.6987 0.1086 -0.6987>}
 dyn_neck_04 {parent=neck_01, loc=<7.3027 0.661 -7.652>, rot=<-0.1086 -0.6987 0.1086 -0.6987>}
 sleepingBag_head {parent=spine_05, loc=<22.5119 4.3132 0>, rot=<0 0 -0.0401 -0.9992>}
 face_head {parent=sleepingBag_head, loc=<10.0163 6.1317 0>, rot=<0.2769 0.6506 0.6506 0.2769>}
 kin_cover_belt_l {parent=face_head, loc=<15.0547 -9.3018 -11.4005>, rot=<-0.6451 -0.2896 0.2896 -0.6451>}
 kin_cover_belt_r {parent=face_head, loc=<15.0543 9.3019 -11.4001>, rot=<-0.6451 -0.2896 0.2896 -0.6451>}
 kinEye_l {parent=face_head, loc=<9.7825 -4.8518 -2.0014>, rot=<-0.2642 0.2642 0.6559 -0.6559>}
 kinEye_r {parent=face_head, loc=<9.7996 4.8663 -2.0226>, rot=<-0.2642 0.2642 0.6559 -0.6559>}
 kinEyebrow_l {parent=face_head, loc=<5.3921 -5.3121 0.5094>, rot=<-0.2642 0.2642 0.6559 -0.6559>}
 kinEyebrow_r {parent=face_head, loc=<5.3921 5.3121 0.5094>, rot=<-0.2642 0.2642 0.6559 -0.6559>}
 kin_cover_belt_r_up_null {parent=face_head, type=Null, loc=<16.4622 9.3019 -10.0466>, rot=<0.6451 0.2896 -0.2896 0.6451>}
 kin_cover_belt_l_up_null {parent=face_head, type=Null, loc=<16.4622 -9.3018 -10.0466>, rot=<0.6451 0.2896 -0.2896 0.6451>}
 kin_coverHead_ik {parent=sleepingBag_head, loc=<12.7523 -2.5884 0.0015>, rot=<0.6729 0.2191 0.2193 0.6716>}
 kin_coverHead_1 {parent=spine_05, loc=<14.546 -5.6818 -0.0039>, rot=<-0.7073 0.0244 0.0248 -0.7061>}
 kin_coverHead_2 {parent=kin_coverHead_1, loc=<10.2742 0 0>, rot=<0 -0.3803 0 -0.9249>}
 kin_coverHead_3 {parent=kin_coverHead_2, loc=<14.0311 0 0>, rot=<0 0 0 1>}
 base_backHandle {parent=kin_coverHead_1, loc=<9.1165 0.0126 1.9054>, rot=<0.0011 0 0 -1>}
 dyn_backHandle {parent=base_backHandle, loc=<0 0 0>, rot=<0.0848 -0.7022 0.7019 0.0843>}
 base_topLid {parent=spine_05, loc=<10.7157 9.7715 0.0056>, rot=<-0.0389 -0.9992 0 0>}
 dyn_topLid {parent=base_topLid, loc=<0 0 0>, rot=<0 0 0 1>}
 base_coverTip_belt_r_1 {parent=dyn_topLid, loc=<10.931 1.6086 -9.3353>, rot=<-0.0128 -0.0063 0.1191 -0.9928>}
 dyn_coverTip_belt_r_1 {parent=base_coverTip_belt_r_1, loc=<0 0 0>, rot=<0 0 0 1>}
 dyn_coverTip_belt_r_2 {parent=dyn_coverTip_belt_r_1, loc=<5.5911 0 0>, rot=<0 0 -0.0597 -0.9982>}
 base_coverTip_belt_l_1 {parent=dyn_topLid, loc=<11.141 1.4146 9.3398>, rot=<0 0 0.0789 -0.9969>}
 dyn_coverTip_belt_l_1 {parent=base_coverTip_belt_l_1, loc=<0 0 0>, rot=<0 0 0 1>}
 target_cover_belt_r {parent=dyn_topLid, loc=<6.0257 1.4324 -9.2963>, rot=<0.0558 0.7049 0.0558 -0.7049>}
 target_cover_belt_l {parent=dyn_topLid, loc=<6.026 1.4323 9.3074>, rot=<0.0558 0.7049 0.0558 -0.7049>}
 head_null {parent=spine_05, type=Null, loc=<20.4455 6.1834 0>, rot=<0 0 -0.036 -0.9994>}
 clavicle_l {parent=spine_05, loc=<5.4831 -0.1053 -3.2946>, rot=<0.0465 -0.7289 -0.009 -0.6829>}
 upperarm_l {parent=clavicle_l, loc=<15.5111 0 0>, rot=<0.0305 -0.536 -0.0057 -0.8437>}
 lowerarm_l {parent=upperarm_l, loc=<26.6514 0 0>, rot=<0.0391 -0.0099 -0.2453 -0.9686>}
 hand_l {parent=lowerarm_l, loc=<22.3491 0 0>, rot=<-0.7081 -0.0507 0.0975 -0.6975>}
 index_metacarpal_l {parent=hand_l, loc=<1.6535 0.0183 -2.96>, rot=<0.0158 -0.0124 -0.0173 -0.9996>}
 index_01_l {parent=index_metacarpal_l, loc=<8.3008 0.0357 -0.0524>, rot=<-0.0044 0.0485 0.0924 -0.9945>}
 index_02_l {parent=index_01_l, loc=<3.3587 -0.3004 0.0452>, rot=<-0.002 0.0027 0.3225 -0.9466>}
 index_03_l {parent=index_02_l, loc=<2.5578 -0.123 -0.0252>, rot=<-0.002 0.0067 0.2558 -0.9667>}
 middle_metacarpal_l {parent=hand_l, loc=<2.1856 -0.1725 -0.5984>, rot=<0.006 0.0168 -0.0329 -0.9993>}
 middle_01_l {parent=middle_metacarpal_l, loc=<7.3275 -0.2447 0.0129>, rot=<-0.0208 0.0042 0.0858 -0.9961>}
 middle_02_l {parent=middle_01_l, loc=<4.0123 0.1401 0.1175>, rot=<-0.0218 0.0258 0.3512 -0.9357>}
 middle_03_l {parent=middle_02_l, loc=<2.6077 0.0534 0.1373>, rot=<0.0082 0.0188 0.1708 -0.9851>}
 pinky_metacarpal_l {parent=hand_l, loc=<1.7616 -0.9753 3.3814>, rot=<-0.0397 0.0814 -0.069 -0.9935>}
 pinky_01_l {parent=pinky_metacarpal_l, loc=<7.1709 -0.4831 -0.1094>, rot=<-0.0654 -0.0035 0.2242 -0.9723>}
 pinky_02_l {parent=pinky_01_l, loc=<2.8274 0.0893 -0.1043>, rot=<0.0037 0.0011 0.2854 -0.9584>}
 pinky_03_l {parent=pinky_02_l, loc=<2.2046 0.0307 -0.0078>, rot=<-0.0002 0.0098 0.1445 -0.9895>}
 ring_metacarpal_l {parent=hand_l, loc=<2.2074 -0.3625 1.3638>, rot=<-0.0284 0.0722 -0.0458 -0.9959>}
 ring_01_l {parent=ring_metacarpal_l, loc=<7.0295 -0.6197 -0.1446>, rot=<-0.0489 -0.0252 0.1892 -0.9804>}
 ring_02_l {parent=ring_01_l, loc=<3.6174 0.4647 -0.1027>, rot=<0 0 0.2806 -0.9598>}
 ring_03_l {parent=ring_02_l, loc=<2.4152 0.114 0.0199>, rot=<0 0 0.1189 -0.9929>}
 thumb_01_l {parent=hand_l, loc=<1.3118 -0.3191 -2.459>, rot=<0.6546 -0.4042 0.0664 -0.6354>}
 thumb_02_l {parent=thumb_01_l, loc=<5.9336 -0.0448 0.0401>, rot=<-0.0037 0.0213 0.298 -0.9543>}
 thumb_03_l {parent=thumb_02_l, loc=<2.2223 0.3682 -0.0907>, rot=<0.0118 -0.013 0.0066 -0.9998>}
 weapon_l {parent=hand_l, loc=<-0.1974 -3.4111 0.3811>, rot=<0 0 0 1>}
 hand_attach_l {parent=hand_l, loc=<8.5 -3 0>, rot=<0.7071 0.7071 0 0>}
 lowerarm_twist_01_l {parent=lowerarm_l, loc=<17.8549 0 0>, rot=<0 0 0 1>}
 lowerarm_twist_02_l {parent=lowerarm_l, loc=<9.6199 0 0>, rot=<0 0 0 1>}
 upperarm_twist_01_l {parent=upperarm_l, loc=<0.0041 0 0>, rot=<0 0 0 1>}
 deform_deltoid_fr_l {parent=upperarm_twist_01_l, loc=<2.1817 -0.0091 1.1183>, rot=<-0.0069 0.2343 0.0003 -0.9722>}
 upperarm_twist_02_l {parent=upperarm_l, loc=<13.7941 0 0>, rot=<0 0 0 1>}
 deform_pectoral_fr_l {parent=clavicle_l, loc=<9.3878 0.0532 -0.0002>, rot=<0 0 0 1>}
 deform_trap_l {parent=clavicle_l, loc=<2.43 -3.9232 6.2456>, rot=<0 -0.1691 0 -0.9856>}
 clavicle_r {parent=spine_05, loc=<5.4831 -0.1053 3.2946>, rot=<0.7289 0.0465 -0.6829 0.009>}
 upperarm_r {parent=clavicle_r, loc=<-15.5111 0 0>, rot=<0.0305 -0.536 -0.0057 -0.8437>}
 lowerarm_r {parent=upperarm_r, loc=<-26.6515 0 0>, rot=<0.0391 -0.0099 -0.2453 -0.9686>}
 hand_r {parent=lowerarm_r, loc=<-22.3492 0 0>, rot=<-0.7081 -0.0507 0.0975 -0.6975>}
 index_metacarpal_r {parent=hand_r, loc=<-1.6534 -0.0181 2.9601>, rot=<0.0158 -0.0124 -0.0173 -0.9996>}
 index_01_r {parent=index_metacarpal_r, loc=<-8.3008 -0.0357 0.0524>, rot=<-0.0044 0.0485 0.0924 -0.9945>}
 index_02_r {parent=index_01_r, loc=<-3.3589 0.3003 -0.0453>, rot=<-0.002 0.0027 0.3225 -0.9466>}
 index_03_r {parent=index_02_r, loc=<-2.5578 0.1231 0.0251>, rot=<-0.002 0.0067 0.2558 -0.9667>}
 middle_metacarpal_r {parent=hand_r, loc=<-2.1856 0.1726 0.5984>, rot=<0.006 0.0168 -0.0329 -0.9993>}
 middle_01_r {parent=middle_metacarpal_r, loc=<-7.3273 0.2447 -0.0129>, rot=<-0.0208 0.0042 0.0858 -0.9961>}
 middle_02_r {parent=middle_01_r, loc=<-4.0142 -0.1411 -0.07>, rot=<-0.0217 0.0259 0.3512 -0.9357>}
 middle_03_r {parent=middle_02_r, loc=<-2.6074 -0.0534 -0.1312>, rot=<0.0083 0.0188 0.1708 -0.9851>}
 pinky_metacarpal_r {parent=hand_r, loc=<-1.7617 0.9753 -3.3814>, rot=<-0.0397 0.0814 -0.069 -0.9935>}
 pinky_01_r {parent=pinky_metacarpal_r, loc=<-7.1707 0.4832 0.1094>, rot=<-0.0654 -0.0035 0.2242 -0.9723>}
 pinky_02_r {parent=pinky_01_r, loc=<-2.8274 -0.0892 0.1043>, rot=<0.0037 0.0011 0.2854 -0.9584>}
 pinky_03_r {parent=pinky_02_r, loc=<-2.2045 -0.031 0.0078>, rot=<-0.0002 0.0098 0.1445 -0.9895>}
 ring_metacarpal_r {parent=hand_r, loc=<-2.2073 0.3627 -1.3638>, rot=<-0.0284 0.0722 -0.0458 -0.9959>}
 ring_01_r {parent=ring_metacarpal_r, loc=<-7.0297 0.6197 0.1446>, rot=<-0.0489 -0.0252 0.1892 -0.9804>}
 ring_02_r {parent=ring_01_r, loc=<-3.6171 -0.4648 0.1027>, rot=<0 0 0.2806 -0.9598>}
 ring_03_r {parent=ring_02_r, loc=<-2.4153 -0.1141 -0.0199>, rot=<0 0 0.1189 -0.9929>}
 thumb_01_r {parent=hand_r, loc=<-1.3119 0.3192 2.459>, rot=<0.6546 -0.4042 0.0664 -0.6354>}
 thumb_02_r {parent=thumb_01_r, loc=<-5.9335 0.0447 -0.0401>, rot=<-0.0037 0.0213 0.298 -0.9543>}
 thumb_03_r {parent=thumb_02_r, loc=<-2.2223 -0.3682 0.0907>, rot=<0.0118 -0.013 0.0066 -0.9998>}
 weapon_r {parent=hand_r, loc=<0.1975 3.4112 -0.3811>, rot=<0 0 0 -1>}
 hand_attach_r {parent=hand_r, loc=<-8.5001 2.9999 0>, rot=<0 0 -0.7071 -0.7071>}
 lowerarm_twist_01_r {parent=lowerarm_r, loc=<-17.855 0 0>, rot=<0 0 0 1>}
 lowerarm_twist_02_r {parent=lowerarm_r, loc=<-9.6199 0 0>, rot=<0 0 0 1>}
 upperarm_twist_01_r {parent=upperarm_r, loc=<-0.0041 0 0>, rot=<0 0 0 1>}
 deform_deltoid_fr_r {parent=upperarm_twist_01_r, loc=<-2.1817 0.0091 -1.1183>, rot=<-0.0069 0.2343 0.0003 -0.9722>}
 upperarm_twist_02_r {parent=upperarm_r, loc=<-13.7942 0 0>, rot=<0 0 0 1>}
 deform_pectoral_fr_r {parent=clavicle_r, loc=<-9.3873 -0.0532 0.0002>, rot=<0 0 0 1>}
 deform_trap_r {parent=clavicle_r, loc=<-2.4278 3.9076 -6.2667>, rot=<0 -0.1691 0 -0.9856>}
 attach_cape {parent=spine_05, loc=<9.017 -11.714 0>, rot=<0.0616 -0.7044 -0.0616 -0.7044>}
 attach_backpack {parent=spine_05, loc=<-0.123 -13.473 0>, rot=<-0.0093 -0.707 0.0093 -0.707>}
 hold_coverHead {parent=spine_05, loc=<0 0 0>, rot=<0.0284 -0.7065 -0.0284 -0.7065>}
 Charm_Attach_00 {parent=spine_05, type=Null, loc=<-0.095 14.4 -7.75>, rot=<-0.5245 0.4379 -0.6186 0.388>}
 thigh_l {parent=pelvis, loc=<-2.5693 -0.0109 -7.7501>, rot=<0.0012 0.0572 0.0155 -0.9982>}
 calf_l {parent=thigh_l, loc=<-39.0484 0 0>, rot=<0.0006 -0.0161 -0.0391 -0.9991>}
 calf_twist_01_l {parent=calf_l, loc=<-35.813 -0.0967 0.0397>, rot=<0.0237 0.0095 -0.0035 -0.9997>}
 calf_twist_02_l {parent=calf_l, loc=<-18.8064 -0.0967 0.0397>, rot=<0.0237 0.0095 -0.0035 -0.9997>}
 foot_l {parent=calf_l, loc=<-42.561 0.1077 0.4292>, rot=<0.0418 -0.0455 0.0343 -0.9975>}
 ball_l {parent=foot_l, loc=<-6.1599 13.8264 0.6806>, rot=<0.0001 -0.0001 -0.7063 -0.7079>}
 thigh_twist_01_l {parent=thigh_l, loc=<-21.6303 0 0>, rot=<0.0013 0 0 -1>}
 thigh_r {parent=pelvis, loc=<-2.5691 -0.0109 7.7501>, rot=<0.0572 -0.0012 0.9982 0.0155>}
 calf_r {parent=thigh_r, loc=<39.0487 0 0>, rot=<0.0006 -0.0161 -0.0391 -0.9991>}
 calf_twist_01_r {parent=calf_r, loc=<35.8132 0.0967 -0.0397>, rot=<0.0237 0.0095 -0.0035 -0.9997>}
 calf_twist_02_r {parent=calf_r, loc=<18.8064 0.0967 -0.0397>, rot=<0.0237 0.0095 -0.0035 -0.9997>}
 foot_r {parent=calf_r, loc=<42.5612 -0.1077 -0.4292>, rot=<0.0418 -0.0455 0.0343 -0.9975>}
 ball_r {parent=foot_r, loc=<6.1598 -13.8264 -0.6806>, rot=<0.0001 -0.0001 -0.7063 -0.7079>}
 thigh_twist_01_r {parent=thigh_r, loc=<21.6303 0 0>, rot=<0.0013 0 0 -1>}
 attach {parent=root, loc=<0 0 0>, rot=<0 0 0 1>}
 ik_foot_root {parent=root, loc=<0 0 0>, rot=<0 0 0 1>}
 ik_foot_l {parent=ik_foot_root, loc=<15.2608 -4.4176 9.4433>, rot=<0.0357 0.7045 0.0284 -0.7082>}
 ik_foot_r {parent=ik_foot_root, loc=<-15.2609 -4.4176 9.4431>, rot=<0.7082 0.0284 -0.7045 0.0357>}
 ik_hand_root {parent=root, loc=<0 0 0>, rot=<0 0 0 1>}
 ik_hand_gun {parent=ik_hand_root, loc=<-33.8029 3.8283 90.5694>, rot=<0.4834 0.3223 0.5887 -0.5621>}
 ik_hand_l {parent=ik_hand_gun, loc=<6.7067 65.8041 13.978>, rot=<-0.0992 -0.9733 -0.2068 0>}
 ik_hand_r {parent=ik_hand_gun, loc=<0 0 0>, rot=<0 0 0 1>}
 attach_fp {parent=root, loc=<0 0 0>, rot=<0 0 0 1>}
 camera_root {parent=root, loc=<0 0 0>, rot=<0 0 0 1>}
 camera {parent=camera_root, loc=<0 0 0>, rot=<0 0 0 1>}
 attach_emote_01 {parent=root, loc=<0 0 0>, rot=<0 0 0 1>}
 attach_emote_02 {parent=root, loc=<0 0 0>, rot=<0 0 0 1>}
 attach_emote_03 {parent=root, loc=<0 0 0>, rot=<0 0 0 1>}
 warp_point_01 {parent=root, loc=<0 0 0>, rot=<0 0 0 1>}
 warp_point_02 {parent=root, loc=<0 0 0>, rot=<0 0 0 1>}
 warp_point_03 {parent=root, loc=<0 0 0>, rot=<0 0 0 1>}
 look_left_pose {type=Curve, loc=<0 0 0>, rot=<0 0 0 1>}
 look_right_pose {type=Curve, loc=<0 0 0>, rot=<0 0 0 1>}
 look_up_pose {type=Curve, loc=<0 0 0>, rot=<0 0 0 1>}
 look_down_pose {type=Curve, loc=<0 0 0>, rot=<0 0 0 1>}
 L_blink_pose {type=Curve, loc=<0 0 0>, rot=<0 0 0 1>}
 R_blink_pose {type=Curve, loc=<0 0 0>, rot=<0 0 0 1>}
 L_wide_pose {type=Curve, loc=<0 0 0>, rot=<0 0 0 1>}
 R_wide_pose {type=Curve, loc=<0 0 0>, rot=<0 0 0 1>}
 R_squint_inner_pose {type=Curve, loc=<0 0 0>, rot=<0 0 0 1>}
 R_squeeze_pose {type=Curve, loc=<0 0 0>, rot=<0 0 0 1>}
 L_squint_inner_pose {type=Curve, loc=<0 0 0>, rot=<0 0 0 1>}
 L_squeeze_pose {type=Curve, loc=<0 0 0>, rot=<0 0 0 1>}
 L_brow_up_pose {type=Curve, loc=<0 0 0>, rot=<0 0 0 1>}
 L_brow_down_pose {type=Curve, loc=<0 0 0>, rot=<0 0 0 1>}
 R_brow_up_pose {type=Curve, loc=<0 0 0>, rot=<0 0 0 1>}
 R_brow_down_pose {type=Curve, loc=<0 0 0>, rot=<0 0 0 1>}
 C_glabella_up_pose {type=Curve, loc=<0 0 0>, rot=<0 0 0 1>}
 C_glabella_down_pose {type=Curve, loc=<0 0 0>, rot=<0 0 0 1>}
 L_brow_squeeze_in_pose {type=Curve, loc=<0 0 0>, rot=<0 0 0 1>}
 R_brow_squeeze_in_pose {type=Curve, loc=<0 0 0>, rot=<0 0 0 1>}
 L_upper_lip_raiser_pose {type=Curve, loc=<0 0 0>, rot=<0 0 0 1>}
 R_upper_lip_raiser_pose {type=Curve, loc=<0 0 0>, rot=<0 0 0 1>}
 L_upper_lip_lower_pose {type=Curve, loc=<0 0 0>, rot=<0 0 0 1>}
 R_upper_lip_lower_pose {type=Curve, loc=<0 0 0>, rot=<0 0 0 1>}
 L_lower_lip_up_pose {type=Curve, loc=<0 0 0>, rot=<0 0 0 1>}
 R_lower_lip_up_pose {type=Curve, loc=<0 0 0>, rot=<0 0 0 1>}
 L_lower_lip_down_pose {type=Curve, loc=<0 0 0>, rot=<0 0 0 1>}
 R_lower_lip_down_pose {type=Curve, loc=<0 0 0>, rot=<0 0 0 1>}
 L_lip_corner_narrow_pose {type=Curve, loc=<0 0 0>, rot=<0 0 0 1>}
 L_lip_corner_wide_pose {type=Curve, loc=<0 0 0>, rot=<0 0 0 1>}
 R_lip_corner_narrow_pose {type=Curve, loc=<0 0 0>, rot=<0 0 0 1>}
 R_lip_corner_wide_pose {type=Curve, loc=<0 0 0>, rot=<0 0 0 1>}
 L_smile_pose {type=Curve, loc=<0 0 0>, rot=<0 0 0 1>}
 R_smile_pose {type=Curve, loc=<0 0 0>, rot=<0 0 0 1>}
 L_frown_pose {type=Curve, loc=<0 0 0>, rot=<0 0 0 1>}
 R_frown_pose {type=Curve, loc=<0 0 0>, rot=<0 0 0 1>}
 phoneme_oo_pose {type=Curve, loc=<0 0 0>, rot=<0 0 0 1>}
 phoneme_ch_pose {type=Curve, loc=<0 0 0>, rot=<0 0 0 1>}
 phoneme_mbp_pose {type=Curve, loc=<0 0 0>, rot=<0 0 0 1>}
 phoneme_fv_pose {type=Curve, loc=<0 0 0>, rot=<0 0 0 1>}
 mouth_up_pose {type=Curve, loc=<0 0 0>, rot=<0 0 0 1>}
 mouth_down_pose {type=Curve, loc=<0 0 0>, rot=<0 0 0 1>}
 C_nose_wrinkler_pose {type=Curve, loc=<0 0 0>, rot=<0 0 0 1>}
 L_cheek_up_pose {type=Curve, loc=<0 0 0>, rot=<0 0 0 1>}
 R_cheek_up_pose {type=Curve, loc=<0 0 0>, rot=<0 0 0 1>}
 jaw_open_pose {type=Curve, loc=<0 0 0>, rot=<0 0 0 1>}
 jaw_right_pose {type=Curve, loc=<0 0 0>, rot=<0 0 0 1>}
 jaw_left_pose {type=Curve, loc=<0 0 0>, rot=<0 0 0 1>}
 eye_convergence_pose {type=Curve, loc=<0 0 0>, rot=<0 0 0 1>}
 L_upperLidDown_pose {type=Curve, loc=<0 0 0>, rot=<0 0 0 1>}
 R_upperLidDown_pose {type=Curve, loc=<0 0 0>, rot=<0 0 0 1>}
 jaw_up_pose {type=Curve, loc=<0 0 0>, rot=<0 0 0 1>}
 jaw_forward_pose {type=Curve, loc=<0 0 0>, rot=<0 0 0 1>}
 jaw_back_pose {type=Curve, loc=<0 0 0>, rot=<0 0 0 1>}
 tongue_forward_pose {type=Curve, loc=<0 0 0>, rot=<0 0 0 1>}
 tongue_back_pose {type=Curve, loc=<0 0 0>, rot=<0 0 0 1>}
 tongue_up_pose {type=Curve, loc=<0 0 0>, rot=<0 0 0 1>}
 tongue_down_pose {type=Curve, loc=<0 0 0>, rot=<0 0 0 1>}
 base_pose {type=Curve, loc=<0 0 0>, rot=<0 0 0 1>}
 Hair_Scarf {type=Curve, loc=<0 0 0>, rot=<0 0 0 1>}
 Hair_Hat {type=Curve, loc=<0 0 0>, rot=<0 0 0 1>}
 Hair_Helmet {type=Curve, loc=<0 0 0>, rot=<0 0 0 1>}
 Hair_Glasses {type=Curve, loc=<0 0 0>, rot=<0 0 0 1>}
 Hair_Cap {type=Curve, loc=<0 0 0>, rot=<0 0 0 1>}
 Hair_Headsets {type=Curve, loc=<0 0 0>, rot=<0 0 0 1>}
 Hair_Hood {type=Curve, loc=<0 0 0>, rot=<0 0 0 1>}
 Face_Mask {type=Curve, loc=<0 0 0>, rot=<0 0 0 1>}
 HairStrands_Hat {type=Curve, loc=<0 0 0>, rot=<0 0 0 1>}
 Hat_Hair {type=Curve, loc=<0 0 0>, rot=<0 0 0 1>}
 Helmet_Hair {type=Curve, loc=<0 0 0>, rot=<0 0 0 1>}
 Hair_Muff {type=Curve, loc=<0 0 0>, rot=<0 0 0 1>}
 Hair_SpaceSuit {type=Curve, loc=<0 0 0>, rot=<0 0 0 1>}
 Hair_Hide {type=Curve, loc=<0 0 0>, rot=<0 0 0 1>}
 Head_SpaceSuit {type=Curve, loc=<0 0 0>, rot=<0 0 0 1>}
 Hair_Null {type=Curve, loc=<0 0 0>, rot=<0 0 0 1>}
 DisableFaceOverride {type=Curve, loc=<0 0 0>, rot=<0 0 0 1>}
 LiveLink {type=Curve, loc=<0 0 0>, rot=<0 0 0 1>}
 Hat_MESH {type=Curve, loc=<0 0 0>, rot=<0 0 0 1>}
 Null_MESH {type=Curve, loc=<0 0 0>, rot=<0 0 0 1>}
 Pose_0 {type=Curve, loc=<0 0 0>, rot=<0 0 0 1>}
 Hair_Mask {type=Curve, loc=<0 0 0>, rot=<0 0 0 1>}
 eye_convergence {type=Curve, loc=<0 0 0>, rot=<0 0 0 1>}
 headTransform {type=Curve, loc=<0 0 0>, rot=<0 0 0 1>}
 Pose_1 {type=Curve, loc=<0 0 0>, rot=<0 0 0 1>}
 Pose_2 {type=Curve, loc=<0 0 0>, rot=<0 0 0 1>}
 Pose_3 {type=Curve, loc=<0 0 0>, rot=<0 0 0 1>}
 Pose_4 {type=Curve, loc=<0 0 0>, rot=<0 0 0 1>}
 Pose_5 {type=Curve, loc=<0 0 0>, rot=<0 0 0 1>}
 Pose_6 {type=Curve, loc=<0 0 0>, rot=<0 0 0 1>}
 Pose_7 {type=Curve, loc=<0 0 0>, rot=<0 0 0 1>}
 Pose_8 {type=Curve, loc=<0 0 0>, rot=<0 0 0 1>}
 Pose_9 {type=Curve, loc=<0 0 0>, rot=<0 0 0 1>}
 Pose_10 {type=Curve, loc=<0 0 0>, rot=<0 0 0 1>}
 Pose_11 {type=Curve, loc=<0 0 0>, rot=<0 0 0 1>}
 Pose_12 {type=Curve, loc=<0 0 0>, rot=<0 0 0 1>}
 Pose_13 {type=Curve, loc=<0 0 0>, rot=<0 0 0 1>}
 Pose_14 {type=Curve, loc=<0 0 0>, rot=<0 0 0 1>}
 Pose_15 {type=Curve, loc=<0 0 0>, rot=<0 0 0 1>}
 Pose_16 {type=Curve, loc=<0 0 0>, rot=<0 0 0 1>}
 Pose_17 {type=Curve, loc=<0 0 0>, rot=<0 0 0 1>}
 Pose_18 {type=Curve, loc=<0 0 0>, rot=<0 0 0 1>}
 Pose_19 {type=Curve, loc=<0 0 0>, rot=<0 0 0 1>}
 Pose_20 {type=Curve, loc=<0 0 0>, rot=<0 0 0 1>}
 Pose_21 {type=Curve, loc=<0 0 0>, rot=<0 0 0 1>}
 Pose_22 {type=Curve, loc=<0 0 0>, rot=<0 0 0 1>}
 Pose_23 {type=Curve, loc=<0 0 0>, rot=<0 0 0 1>}
 Pose_24 {type=Curve, loc=<0 0 0>, rot=<0 0 0 1>}
 Pose_25 {type=Curve, loc=<0 0 0>, rot=<0 0 0 1>}
 Pose_26 {type=Curve, loc=<0 0 0>, rot=<0 0 0 1>}
 Pose_27 {type=Curve, loc=<0 0 0>, rot=<0 0 0 1>}
 Pose_28 {type=Curve, loc=<0 0 0>, rot=<0 0 0 1>}
 Pose_29 {type=Curve, loc=<0 0 0>, rot=<0 0 0 1>}
 Pose_30 {type=Curve, loc=<0 0 0>, rot=<0 0 0 1>}
 Pose_31 {type=Curve, loc=<0 0 0>, rot=<0 0 0 1>}
 Pose_32 {type=Curve, loc=<0 0 0>, rot=<0 0 0 1>}
 Pose_33 {type=Curve, loc=<0 0 0>, rot=<0 0 0 1>}
 Pose_34 {type=Curve, loc=<0 0 0>, rot=<0 0 0 1>}
 Pose_35 {type=Curve, loc=<0 0 0>, rot=<0 0 0 1>}
 Pose_36 {type=Curve, loc=<0 0 0>, rot=<0 0 0 1>}
 Pose_37 {type=Curve, loc=<0 0 0>, rot=<0 0 0 1>}
 Pose_38 {type=Curve, loc=<0 0 0>, rot=<0 0 0 1>}
 Pose_39 {type=Curve, loc=<0 0 0>, rot=<0 0 0 1>}
 Pose_40 {type=Curve, loc=<0 0 0>, rot=<0 0 0 1>}
 Pose_41 {type=Curve, loc=<0 0 0>, rot=<0 0 0 1>}
 Pose_42 {type=Curve, loc=<0 0 0>, rot=<0 0 0 1>}
 Pose_43 {type=Curve, loc=<0 0 0>, rot=<0 0 0 1>}
 Pose_44 {type=Curve, loc=<0 0 0>, rot=<0 0 0 1>}
 Pose_45 {type=Curve, loc=<0 0 0>, rot=<0 0 0 1>}
 Pose_46 {type=Curve, loc=<0 0 0>, rot=<0 0 0 1>}
 Pose_47 {type=Curve, loc=<0 0 0>, rot=<0 0 0 1>}
 Pose_48 {type=Curve, loc=<0 0 0>, rot=<0 0 0 1>}
 Pose_49 {type=Curve, loc=<0 0 0>, rot=<0 0 0 1>}
 Pose_50 {type=Curve, loc=<0 0 0>, rot=<0 0 0 1>}
 Pose_51 {type=Curve, loc=<0 0 0>, rot=<0 0 0 1>}
 Pose_52 {type=Curve, loc=<0 0 0>, rot=<0 0 0 1>}
 Pose_53 {type=Curve, loc=<0 0 0>, rot=<0 0 0 1>}
 Pose_54 {type=Curve, loc=<0 0 0>, rot=<0 0 0 1>}
 Pose_55 {type=Curve, loc=<0 0 0>, rot=<0 0 0 1>}
 Pose_56 {type=Curve, loc=<0 0 0>, rot=<0 0 0 1>}
 Pose_57 {type=Curve, loc=<0 0 0>, rot=<0 0 0 1>}
 Pose_58 {type=Curve, loc=<0 0 0>, rot=<0 0 0 1>}
 Pose_59 {type=Curve, loc=<0 0 0>, rot=<0 0 0 1>}
 Pose_60 {type=Curve, loc=<0 0 0>, rot=<0 0 0 1>}
 Pose_61 {type=Curve, loc=<0 0 0>, rot=<0 0 0 1>}
 Pose_62 {type=Curve, loc=<0 0 0>, rot=<0 0 0 1>}
 Pose_63 {type=Curve, loc=<0 0 0>, rot=<0 0 0 1>}
 Pose_64 {type=Curve, loc=<0 0 0>, rot=<0 0 0 1>}
 Pose_65 {type=Curve, loc=<0 0 0>, rot=<0 0 0 1>}
 Pose_66 {type=Curve, loc=<0 0 0>, rot=<0 0 0 1>}
 Pose_67 {type=Curve, loc=<0 0 0>, rot=<0 0 0 1>}
 Pose_68 {type=Curve, loc=<0 0 0>, rot=<0 0 0 1>}
 Pose_69 {type=Curve, loc=<0 0 0>, rot=<0 0 0 1>}
 Pose_70 {type=Curve, loc=<0 0 0>, rot=<0 0 0 1>}
 Pose_71 {type=Curve, loc=<0 0 0>, rot=<0 0 0 1>}
 Pose_72 {type=Curve, loc=<0 0 0>, rot=<0 0 0 1>}
 Pose_73 {type=Curve, loc=<0 0 0>, rot=<0 0 0 1>}
 Pose_74 {type=Curve, loc=<0 0 0>, rot=<0 0 0 1>}
 Pose_75 {type=Curve, loc=<0 0 0>, rot=<0 0 0 1>}
 Pose_76 {type=Curve, loc=<0 0 0>, rot=<0 0 0 1>}
 Pose_77 {type=Curve, loc=<0 0 0>, rot=<0 0 0 1>}
 Pose_78 {type=Curve, loc=<0 0 0>, rot=<0 0 0 1>}
 Pose_79 {type=Curve, loc=<0 0 0>, rot=<0 0 0 1>}
 Pose_80 {type=Curve, loc=<0 0 0>, rot=<0 0 0 1>}
 Pose_81 {type=Curve, loc=<0 0 0>, rot=<0 0 0 1>}
 Pose_82 {type=Curve, loc=<0 0 0>, rot=<0 0 0 1>}
 Pose_83 {type=Curve, loc=<0 0 0>, rot=<0 0 0 1>}
 Pose_84 {type=Curve, loc=<0 0 0>, rot=<0 0 0 1>}
 Pose_85 {type=Curve, loc=<0 0 0>, rot=<0 0 0 1>}
 Pose_86 {type=Curve, loc=<0 0 0>, rot=<0 0 0 1>}
 Pose_87 {type=Curve, loc=<0 0 0>, rot=<0 0 0 1>}
 Pose_88 {type=Curve, loc=<0 0 0>, rot=<0 0 0 1>}
 Pose_89 {type=Curve, loc=<0 0 0>, rot=<0 0 0 1>}
 Pose_90 {type=Curve, loc=<0 0 0>, rot=<0 0 0 1>}
 Pose_91 {type=Curve, loc=<0 0 0>, rot=<0 0 0 1>}
 Pose_92 {type=Curve, loc=<0 0 0>, rot=<0 0 0 1>}
 Pose_93 {type=Curve, loc=<0 0 0>, rot=<0 0 0 1>}
 Pose_94 {type=Curve, loc=<0 0 0>, rot=<0 0 0 1>}
 Pose_95 {type=Curve, loc=<0 0 0>, rot=<0 0 0 1>}
 Pose_96 {type=Curve, loc=<0 0 0>, rot=<0 0 0 1>}
 Pose_97 {type=Curve, loc=<0 0 0>, rot=<0 0 0 1>}
 Pose_98 {type=Curve, loc=<0 0 0>, rot=<0 0 0 1>}
 Pose_99 {type=Curve, loc=<0 0 0>, rot=<0 0 0 1>}
 Pose_100 {type=Curve, loc=<0 0 0>, rot=<0 0 0 1>}
 Pose_101 {type=Curve, loc=<0 0 0>, rot=<0 0 0 1>}
 Pose_102 {type=Curve, loc=<0 0 0>, rot=<0 0 0 1>}
 Pose_103 {type=Curve, loc=<0 0 0>, rot=<0 0 0 1>}
 Pose_104 {type=Curve, loc=<0 0 0>, rot=<0 0 0 1>}
 Pose_105 {type=Curve, loc=<0 0 0>, rot=<0 0 0 1>}
 Pose_106 {type=Curve, loc=<0 0 0>, rot=<0 0 0 1>}
 Pose_107 {type=Curve, loc=<0 0 0>, rot=<0 0 0 1>}
 Pose_108 {type=Curve, loc=<0 0 0>, rot=<0 0 0 1>}
 Pose_109 {type=Curve, loc=<0 0 0>, rot=<0 0 0 1>}
 Pose_110 {type=Curve, loc=<0 0 0>, rot=<0 0 0 1>}
 Pose_111 {type=Curve, loc=<0 0 0>, rot=<0 0 0 1>}
 Pose_112 {type=Curve, loc=<0 0 0>, rot=<0 0 0 1>}
 Pose_113 {type=Curve, loc=<0 0 0>, rot=<0 0 0 1>}
 Pose_114 {type=Curve, loc=<0 0 0>, rot=<0 0 0 1>}
 Pose_115 {type=Curve, loc=<0 0 0>, rot=<0 0 0 1>}
 Pose_116 {type=Curve, loc=<0 0 0>, rot=<0 0 0 1>}
 Pose_117 {type=Curve, loc=<0 0 0>, rot=<0 0 0 1>}
 Pose_118 {type=Curve, loc=<0 0 0>, rot=<0 0 0 1>}
 Pose_119 {type=Curve, loc=<0 0 0>, rot=<0 0 0 1>}
 Pose_120 {type=Curve, loc=<0 0 0>, rot=<0 0 0 1>}
 Pose_121 {type=Curve, loc=<0 0 0>, rot=<0 0 0 1>}
 Pose_122 {type=Curve, loc=<0 0 0>, rot=<0 0 0 1>}
 Pose_123 {type=Curve, loc=<0 0 0>, rot=<0 0 0 1>}
 Pose_124 {type=Curve, loc=<0 0 0>, rot=<0 0 0 1>}
 Pose_125 {type=Curve, loc=<0 0 0>, rot=<0 0 0 1>}
 Pose_126 {type=Curve, loc=<0 0 0>, rot=<0 0 0 1>}
 Pose_127 {type=Curve, loc=<0 0 0>, rot=<0 0 0 1>}
 Pose_128 {type=Curve, loc=<0 0 0>, rot=<0 0 0 1>}
 Pose_129 {type=Curve, loc=<0 0 0>, rot=<0 0 0 1>}
 Pose_130 {type=Curve, loc=<0 0 0>, rot=<0 0 0 1>}
 Pose_131 {type=Curve, loc=<0 0 0>, rot=<0 0 0 1>}
 Pose_132 {type=Curve, loc=<0 0 0>, rot=<0 0 0 1>}
 Pose_133 {type=Curve, loc=<0 0 0>, rot=<0 0 0 1>}
 Pose_134 {type=Curve, loc=<0 0 0>, rot=<0 0 0 1>}
 Pose_135 {type=Curve, loc=<0 0 0>, rot=<0 0 0 1>}
 Pose_136 {type=Curve, loc=<0 0 0>, rot=<0 0 0 1>}
 Pose_137 {type=Curve, loc=<0 0 0>, rot=<0 0 0 1>}
 Pose_138 {type=Curve, loc=<0 0 0>, rot=<0 0 0 1>}
 Pose_139 {type=Curve, loc=<0 0 0>, rot=<0 0 0 1>}
 Pose_140 {type=Curve, loc=<0 0 0>, rot=<0 0 0 1>}
 Pose_141 {type=Curve, loc=<0 0 0>, rot=<0 0 0 1>}
 Pose_142 {type=Curve, loc=<0 0 0>, rot=<0 0 0 1>}
 Pose_143 {type=Curve, loc=<0 0 0>, rot=<0 0 0 1>}
 Pose_144 {type=Curve, loc=<0 0 0>, rot=<0 0 0 1>}
 Pose_145 {type=Curve, loc=<0 0 0>, rot=<0 0 0 1>}
 Pose_146 {type=Curve, loc=<0 0 0>, rot=<0 0 0 1>}
 Pose_147 {type=Curve, loc=<0 0 0>, rot=<0 0 0 1>}
 Pose_148 {type=Curve, loc=<0 0 0>, rot=<0 0 0 1>}
 Pose_149 {type=Curve, loc=<0 0 0>, rot=<0 0 0 1>}
 Pose_150 {type=Curve, loc=<0 0 0>, rot=<0 0 0 1>}
 Pose_151 {type=Curve, loc=<0 0 0>, rot=<0 0 0 1>}
 Pose_152 {type=Curve, loc=<0 0 0>, rot=<0 0 0 1>}
 Pose_153 {type=Curve, loc=<0 0 0>, rot=<0 0 0 1>}
 Pose_154 {type=Curve, loc=<0 0 0>, rot=<0 0 0 1>}
 Pose_155 {type=Curve, loc=<0 0 0>, rot=<0 0 0 1>}
 Pose_156 {type=Curve, loc=<0 0 0>, rot=<0 0 0 1>}
 Pose_157 {type=Curve, loc=<0 0 0>, rot=<0 0 0 1>}
 Pose_158 {type=Curve, loc=<0 0 0>, rot=<0 0 0 1>}
 Pose_159 {type=Curve, loc=<0 0 0>, rot=<0 0 0 1>}
 Pose_160 {type=Curve, loc=<0 0 0>, rot=<0 0 0 1>}
 Pose_161 {type=Curve, loc=<0 0 0>, rot=<0 0 0 1>}
 Pose_162 {type=Curve, loc=<0 0 0>, rot=<0 0 0 1>}
 Pose_163 {type=Curve, loc=<0 0 0>, rot=<0 0 0 1>}
 Pose_164 {type=Curve, loc=<0 0 0>, rot=<0 0 0 1>}
 Pose_165 {type=Curve, loc=<0 0 0>, rot=<0 0 0 1>}
 Pose_166 {type=Curve, loc=<0 0 0>, rot=<0 0 0 1>}
 Pose_167 {type=Curve, loc=<0 0 0>, rot=<0 0 0 1>}
 Pose_168 {type=Curve, loc=<0 0 0>, rot=<0 0 0 1>}
 Pose_169 {type=Curve, loc=<0 0 0>, rot=<0 0 0 1>}
 Pose_170 {type=Curve, loc=<0 0 0>, rot=<0 0 0 1>}
 Pose_171 {type=Curve, loc=<0 0 0>, rot=<0 0 0 1>}
 Pose_172 {type=Curve, loc=<0 0 0>, rot=<0 0 0 1>}
 Pose_173 {type=Curve, loc=<0 0 0>, rot=<0 0 0 1>}
 Pose_174 {type=Curve, loc=<0 0 0>, rot=<0 0 0 1>}
 Pose_175 {type=Curve, loc=<0 0 0>, rot=<0 0 0 1>}
 Pose_176 {type=Curve, loc=<0 0 0>, rot=<0 0 0 1>}
 Pose_177 {type=Curve, loc=<0 0 0>, rot=<0 0 0 1>}
 Pose_178 {type=Curve, loc=<0 0 0>, rot=<0 0 0 1>}
 Pose_179 {type=Curve, loc=<0 0 0>, rot=<0 0 0 1>}
 eyeLookDownRight {type=Curve, loc=<0 0 0>, rot=<0 0 0 1>}
 eyeLookOutRight {type=Curve, loc=<0 0 0>, rot=<0 0 0 1>}
 eyeLookUpRight {type=Curve, loc=<0 0 0>, rot=<0 0 0 1>}
 eyeLookOutLeft {type=Curve, loc=<0 0 0>, rot=<0 0 0 1>}
 eyeBlinkLeft {type=Curve, loc=<0 0 0>, rot=<0 0 0 1>}
 eyeBlinkRight {type=Curve, loc=<0 0 0>, rot=<0 0 0 1>}
 eyeSquintLeft {type=Curve, loc=<0 0 0>, rot=<0 0 0 1>}
 eyeSquintRight {type=Curve, loc=<0 0 0>, rot=<0 0 0 1>}
 eyeWideLeft {type=Curve, loc=<0 0 0>, rot=<0 0 0 1>}
 eyeWideRight {type=Curve, loc=<0 0 0>, rot=<0 0 0 1>}
 jawForward {type=Curve, loc=<0 0 0>, rot=<0 0 0 1>}
 jawLeft {type=Curve, loc=<0 0 0>, rot=<0 0 0 1>}
 jawRight {type=Curve, loc=<0 0 0>, rot=<0 0 0 1>}
 jawOpen {type=Curve, loc=<0 0 0>, rot=<0 0 0 1>}
 mouthClose {type=Curve, loc=<0 0 0>, rot=<0 0 0 1>}
 mouthFunnel {type=Curve, loc=<0 0 0>, rot=<0 0 0 1>}
 mouthPucker {type=Curve, loc=<0 0 0>, rot=<0 0 0 1>}
 mouthLeft {type=Curve, loc=<0 0 0>, rot=<0 0 0 1>}
 mouthRight {type=Curve, loc=<0 0 0>, rot=<0 0 0 1>}
 mouthSmileLeft {type=Curve, loc=<0 0 0>, rot=<0 0 0 1>}
 mouthSmileRight {type=Curve, loc=<0 0 0>, rot=<0 0 0 1>}
 mouthFrownLeft {type=Curve, loc=<0 0 0>, rot=<0 0 0 1>}
 mouthFrownRight {type=Curve, loc=<0 0 0>, rot=<0 0 0 1>}
 mouthDimpleLeft {type=Curve, loc=<0 0 0>, rot=<0 0 0 1>}
 mouthDimpleRight {type=Curve, loc=<0 0 0>, rot=<0 0 0 1>}
 mouthStretchLeft {type=Curve, loc=<0 0 0>, rot=<0 0 0 1>}
 mouthStretchRight {type=Curve, loc=<0 0 0>, rot=<0 0 0 1>}
 mouthRollLower {type=Curve, loc=<0 0 0>, rot=<0 0 0 1>}
 mouthRollUpper {type=Curve, loc=<0 0 0>, rot=<0 0 0 1>}
 mouthShrugLower {type=Curve, loc=<0 0 0>, rot=<0 0 0 1>}
 mouthShrugUpper {type=Curve, loc=<0 0 0>, rot=<0 0 0 1>}
 mouthPressLeft {type=Curve, loc=<0 0 0>, rot=<0 0 0 1>}
 mouthPressRight {type=Curve, loc=<0 0 0>, rot=<0 0 0 1>}
 mouthLowerDownLeft {type=Curve, loc=<0 0 0>, rot=<0 0 0 1>}
 mouthLowerDownRight {type=Curve, loc=<0 0 0>, rot=<0 0 0 1>}
 mouthUpperUpLeft {type=Curve, loc=<0 0 0>, rot=<0 0 0 1>}
 mouthUpperUpRight {type=Curve, loc=<0 0 0>, rot=<0 0 0 1>}
 browDownLeft {type=Curve, loc=<0 0 0>, rot=<0 0 0 1>}
 browDownRight {type=Curve, loc=<0 0 0>, rot=<0 0 0 1>}
 browInnerUp {type=Curve, loc=<0 0 0>, rot=<0 0 0 1>}
 browOuterUpLeft {type=Curve, loc=<0 0 0>, rot=<0 0 0 1>}
 browOuterUpRight {type=Curve, loc=<0 0 0>, rot=<0 0 0 1>}
 cheekPuff {type=Curve, loc=<0 0 0>, rot=<0 0 0 1>}
 cheekSquintLeft {type=Curve, loc=<0 0 0>, rot=<0 0 0 1>}
 cheekSquintRight {type=Curve, loc=<0 0 0>, rot=<0 0 0 1>}
 noseSneerLeft {type=Curve, loc=<0 0 0>, rot=<0 0 0 1>}
 noseSneerRight {type=Curve, loc=<0 0 0>, rot=<0 0 0 1>}
 C_glabella_down {type=Curve, loc=<0 0 0>, rot=<0 0 0 1>}
 L_brow_squeeze_in {type=Curve, loc=<0 0 0>, rot=<0 0 0 1>}
 R_brow_squeeze_in {type=Curve, loc=<0 0 0>, rot=<0 0 0 1>}
 L_upper_lip_lower {type=Curve, loc=<0 0 0>, rot=<0 0 0 1>}
 R_upper_lip_lower {type=Curve, loc=<0 0 0>, rot=<0 0 0 1>}
 L_lower_lip_up {type=Curve, loc=<0 0 0>, rot=<0 0 0 1>}
 R_lower_lip_up {type=Curve, loc=<0 0 0>, rot=<0 0 0 1>}
 L_lip_corner_narrow {type=Curve, loc=<0 0 0>, rot=<0 0 0 1>}
 R_lip_corner_narrow {type=Curve, loc=<0 0 0>, rot=<0 0 0 1>}
 phoneme_ch {type=Curve, loc=<0 0 0>, rot=<0 0 0 1>}
 phoneme_mbp {type=Curve, loc=<0 0 0>, rot=<0 0 0 1>}
 phoneme_fv {type=Curve, loc=<0 0 0>, rot=<0 0 0 1>}
 mouth_up {type=Curve, loc=<0 0 0>, rot=<0 0 0 1>}
 mouth_down {type=Curve, loc=<0 0 0>, rot=<0 0 0 1>}
 L_cheek_up {type=Curve, loc=<0 0 0>, rot=<0 0 0 1>}
 R_cheek_up {type=Curve, loc=<0 0 0>, rot=<0 0 0 1>}
 L_upperLidDown {type=Curve, loc=<0 0 0>, rot=<0 0 0 1>}
 R_upperLidDown {type=Curve, loc=<0 0 0>, rot=<0 0 0 1>}
 jaw_up {type=Curve, loc=<0 0 0>, rot=<0 0 0 1>}
 jaw_back {type=Curve, loc=<0 0 0>, rot=<0 0 0 1>}
 tongue_forward {type=Curve, loc=<0 0 0>, rot=<0 0 0 1>}
 tongue_back {type=Curve, loc=<0 0 0>, rot=<0 0 0 1>}
 tongue_up {type=Curve, loc=<0 0 0>, rot=<0 0 0 1>}
 tongue_down {type=Curve, loc=<0 0 0>, rot=<0 0 0 1>}
 basePose {type=Curve, loc=<0 0 0>, rot=<0 0 0 1>}
 Pose_180 {type=Curve, loc=<0 0 0>, rot=<0 0 0 1>}
 Pose_181 {type=Curve, loc=<0 0 0>, rot=<0 0 0 1>}
 Pose_182 {type=Curve, loc=<0 0 0>, rot=<0 0 0 1>}
 Pose_183 {type=Curve, loc=<0 0 0>, rot=<0 0 0 1>}
 Pose_184 {type=Curve, loc=<0 0 0>, rot=<0 0 0 1>}
 Pose_185 {type=Curve, loc=<0 0 0>, rot=<0 0 0 1>}
 Pose_186 {type=Curve, loc=<0 0 0>, rot=<0 0 0 1>}
 Pose_187 {type=Curve, loc=<0 0 0>, rot=<0 0 0 1>}
 Pose_188 {type=Curve, loc=<0 0 0>, rot=<0 0 0 1>}
 Pose_189 {type=Curve, loc=<0 0 0>, rot=<0 0 0 1>}
 Pose_190 {type=Curve, loc=<0 0 0>, rot=<0 0 0 1>}
 Pose_191 {type=Curve, loc=<0 0 0>, rot=<0 0 0 1>}
 Pose_192 {type=Curve, loc=<0 0 0>, rot=<0 0 0 1>}
 Pose_193 {type=Curve, loc=<0 0 0>, rot=<0 0 0 1>}
 Pose_194 {type=Curve, loc=<0 0 0>, rot=<0 0 0 1>}
 Pose_195 {type=Curve, loc=<0 0 0>, rot=<0 0 0 1>}
 Pose_196 {type=Curve, loc=<0 0 0>, rot=<0 0 0 1>}
 Pose_197 {type=Curve, loc=<0 0 0>, rot=<0 0 0 1>}
 Pose_198 {type=Curve, loc=<0 0 0>, rot=<0 0 0 1>}
 Pose_199 {type=Curve, loc=<0 0 0>, rot=<0 0 0 1>}
 Pose_200 {type=Curve, loc=<0 0 0>, rot=<0 0 0 1>}
 Pose_201 {type=Curve, loc=<0 0 0>, rot=<0 0 0 1>}
 Pose_202 {type=Curve, loc=<0 0 0>, rot=<0 0 0 1>}
 Pose_203 {type=Curve, loc=<0 0 0>, rot=<0 0 0 1>}
 Pose_204 {type=Curve, loc=<0 0 0>, rot=<0 0 0 1>}
 Pose_205 {type=Curve, loc=<0 0 0>, rot=<0 0 0 1>}
 Pose_206 {type=Curve, loc=<0 0 0>, rot=<0 0 0 1>}
 Pose_207 {type=Curve, loc=<0 0 0>, rot=<0 0 0 1>}
 Pose_208 {type=Curve, loc=<0 0 0>, rot=<0 0 0 1>}
 Pose_209 {type=Curve, loc=<0 0 0>, rot=<0 0 0 1>}
 Pose_210 {type=Curve, loc=<0 0 0>, rot=<0 0 0 1>}
 Pose_211 {type=Curve, loc=<0 0 0>, rot=<0 0 0 1>}
 Pose_212 {type=Curve, loc=<0 0 0>, rot=<0 0 0 1>}
 Pose_213 {type=Curve, loc=<0 0 0>, rot=<0 0 0 1>}
 Pose_214 {type=Curve, loc=<0 0 0>, rot=<0 0 0 1>}
 Pose_215 {type=Curve, loc=<0 0 0>, rot=<0 0 0 1>}
 Pose_216 {type=Curve, loc=<0 0 0>, rot=<0 0 0 1>}
 Pose_217 {type=Curve, loc=<0 0 0>, rot=<0 0 0 1>}
 Pose_218 {type=Curve, loc=<0 0 0>, rot=<0 0 0 1>}
 Pose_219 {type=Curve, loc=<0 0 0>, rot=<0 0 0 1>}
 Pose_220 {type=Curve, loc=<0 0 0>, rot=<0 0 0 1>}
 Pose_221 {type=Curve, loc=<0 0 0>, rot=<0 0 0 1>}
 Pose_222 {type=Curve, loc=<0 0 0>, rot=<0 0 0 1>}
 Pose_223 {type=Curve, loc=<0 0 0>, rot=<0 0 0 1>}
 Pose_224 {type=Curve, loc=<0 0 0>, rot=<0 0 0 1>}
 Pose_225 {type=Curve, loc=<0 0 0>, rot=<0 0 0 1>}
 Pose_226 {type=Curve, loc=<0 0 0>, rot=<0 0 0 1>}
 Pose_227 {type=Curve, loc=<0 0 0>, rot=<0 0 0 1>}
 Pose_228 {type=Curve, loc=<0 0 0>, rot=<0 0 0 1>}
 Pose_229 {type=Curve, loc=<0 0 0>, rot=<0 0 0 1>}
 Pose_230 {type=Curve, loc=<0 0 0>, rot=<0 0 0 1>}
 Pose_231 {type=Curve, loc=<0 0 0>, rot=<0 0 0 1>}
 Pose_232 {type=Curve, loc=<0 0 0>, rot=<0 0 0 1>}
 Pose_233 {type=Curve, loc=<0 0 0>, rot=<0 0 0 1>}
 Pose_234 {type=Curve, loc=<0 0 0>, rot=<0 0 0 1>}
 Pose_235 {type=Curve, loc=<0 0 0>, rot=<0 0 0 1>}
 Pose_236 {type=Curve, loc=<0 0 0>, rot=<0 0 0 1>}
 Pose_237 {type=Curve, loc=<0 0 0>, rot=<0 0 0 1>}
 Pose_238 {type=Curve, loc=<0 0 0>, rot=<0 0 0 1>}
 Pose_239 {type=Curve, loc=<0 0 0>, rot=<0 0 0 1>}
 Pose_240 {type=Curve, loc=<0 0 0>, rot=<0 0 0 1>}
 Pose_241 {type=Curve, loc=<0 0 0>, rot=<0 0 0 1>}
 Pose_242 {type=Curve, loc=<0 0 0>, rot=<0 0 0 1>}
 Pose_243 {type=Curve, loc=<0 0 0>, rot=<0 0 0 1>}
 Pose_244 {type=Curve, loc=<0 0 0>, rot=<0 0 0 1>}
 Pose_245 {type=Curve, loc=<0 0 0>, rot=<0 0 0 1>}
 Pose_246 {type=Curve, loc=<0 0 0>, rot=<0 0 0 1>}
 Pose_247 {type=Curve, loc=<0 0 0>, rot=<0 0 0 1>}
 Pose_248 {type=Curve, loc=<0 0 0>, rot=<0 0 0 1>}
 Pose_249 {type=Curve, loc=<0 0 0>, rot=<0 0 0 1>}
 Pose_250 {type=Curve, loc=<0 0 0>, rot=<0 0 0 1>}
 Pose_251 {type=Curve, loc=<0 0 0>, rot=<0 0 0 1>}
 Pose_252 {type=Curve, loc=<0 0 0>, rot=<0 0 0 1>}
 Pose_253 {type=Curve, loc=<0 0 0>, rot=<0 0 0 1>}
 Pose_254 {type=Curve, loc=<0 0 0>, rot=<0 0 0 1>}
 Pose_255 {type=Curve, loc=<0 0 0>, rot=<0 0 0 1>}
 Pose_256 {type=Curve, loc=<0 0 0>, rot=<0 0 0 1>}
 Pose_257 {type=Curve, loc=<0 0 0>, rot=<0 0 0 1>}
 Pose_258 {type=Curve, loc=<0 0 0>, rot=<0 0 0 1>}
 neutral {type=Curve, loc=<0 0 0>, rot=<0 0 0 1>}
 Pose_259 {type=Curve, loc=<0 0 0>, rot=<0 0 0 1>}
 cheek_puff_low {type=Curve, loc=<0 0 0>, rot=<0 0 0 1>}
 Pose_260 {type=Curve, loc=<0 0 0>, rot=<0 0 0 1>}
 Pose_261 {type=Curve, loc=<0 0 0>, rot=<0 0 0 1>}
 Pose_262 {type=Curve, loc=<0 0 0>, rot=<0 0 0 1>}
 Pose_263 {type=Curve, loc=<0 0 0>, rot=<0 0 0 1>}
 Pose_264 {type=Curve, loc=<0 0 0>, rot=<0 0 0 1>}
 Pose_265 {type=Curve, loc=<0 0 0>, rot=<0 0 0 1>}
 Pose_266 {type=Curve, loc=<0 0 0>, rot=<0 0 0 1>}
 Pose_267 {type=Curve, loc=<0 0 0>, rot=<0 0 0 1>}
 Pose_268 {type=Curve, loc=<0 0 0>, rot=<0 0 0 1>}
 Pose_269 {type=Curve, loc=<0 0 0>, rot=<0 0 0 1>}
 Pose_270 {type=Curve, loc=<0 0 0>, rot=<0 0 0 1>}
 Pose_271 {type=Curve, loc=<0 0 0>, rot=<0 0 0 1>}
 Pose_272 {type=Curve, loc=<0 0 0>, rot=<0 0 0 1>}
 Pose_273 {type=Curve, loc=<0 0 0>, rot=<0 0 0 1>}
 Pose_274 {type=Curve, loc=<0 0 0>, rot=<0 0 0 1>}
 Pose_275 {type=Curve, loc=<0 0 0>, rot=<0 0 0 1>}
 Pose_276 {type=Curve, loc=<0 0 0>, rot=<0 0 0 1>}
 Pose_277 {type=Curve, loc=<0 0 0>, rot=<0 0 0 1>}
 Pose_278 {type=Curve, loc=<0 0 0>, rot=<0 0 0 1>}
 Pose_279 {type=Curve, loc=<0 0 0>, rot=<0 0 0 1>}
 Pose_280 {type=Curve, loc=<0 0 0>, rot=<0 0 0 1>}
 Pose_281 {type=Curve, loc=<0 0 0>, rot=<0 0 0 1>}
 Pose_282 {type=Curve, loc=<0 0 0>, rot=<0 0 0 1>}
 Pose_283 {type=Curve, loc=<0 0 0>, rot=<0 0 0 1>}
 Pose_284 {type=Curve, loc=<0 0 0>, rot=<0 0 0 1>}
 Pose_285 {type=Curve, loc=<0 0 0>, rot=<0 0 0 1>}
 Pose_286 {type=Curve, loc=<0 0 0>, rot=<0 0 0 1>}
 Pose_287 {type=Curve, loc=<0 0 0>, rot=<0 0 0 1>}
 Pose_288 {type=Curve, loc=<0 0 0>, rot=<0 0 0 1>}
 Pose_289 {type=Curve, loc=<0 0 0>, rot=<0 0 0 1>}
 Pose_290 {type=Curve, loc=<0 0 0>, rot=<0 0 0 1>}
 Pose_291 {type=Curve, loc=<0 0 0>, rot=<0 0 0 1>}
 Pose_292 {type=Curve, loc=<0 0 0>, rot=<0 0 0 1>}
 Pose_293 {type=Curve, loc=<0 0 0>, rot=<0 0 0 1>}
 Pose_294 {type=Curve, loc=<0 0 0>, rot=<0 0 0 1>}
 Pose_295 {type=Curve, loc=<0 0 0>, rot=<0 0 0 1>}
 Pose_296 {type=Curve, loc=<0 0 0>, rot=<0 0 0 1>}
 Pose_297 {type=Curve, loc=<0 0 0>, rot=<0 0 0 1>}
 Pose_298 {type=Curve, loc=<0 0 0>, rot=<0 0 0 1>}
 Pose_299 {type=Curve, loc=<0 0 0>, rot=<0 0 0 1>}
 Pose_300 {type=Curve, loc=<0 0 0>, rot=<0 0 0 1>}
 Pose_301 {type=Curve, loc=<0 0 0>, rot=<0 0 0 1>}
 Pose_302 {type=Curve, loc=<0 0 0>, rot=<0 0 0 1>}
 Pose_303 {type=Curve, loc=<0 0 0>, rot=<0 0 0 1>}
 Pose_304 {type=Curve, loc=<0 0 0>, rot=<0 0 0 1>}
 Pose_305 {type=Curve, loc=<0 0 0>, rot=<0 0 0 1>}
 Pose_306 {type=Curve, loc=<0 0 0>, rot=<0 0 0 1>}
 Pose_307 {type=Curve, loc=<0 0 0>, rot=<0 0 0 1>}
 Pose_308 {type=Curve, loc=<0 0 0>, rot=<0 0 0 1>}
 Pose_309 {type=Curve, loc=<0 0 0>, rot=<0 0 0 1>}
 Pose_310 {type=Curve, loc=<0 0 0>, rot=<0 0 0 1>}
 Pose_311 {type=Curve, loc=<0 0 0>, rot=<0 0 0 1>}
 Pose_312 {type=Curve, loc=<0 0 0>, rot=<0 0 0 1>}
 Pose_313 {type=Curve, loc=<0 0 0>, rot=<0 0 0 1>}
 Pose_314 {type=Curve, loc=<0 0 0>, rot=<0 0 0 1>}
 Pose_315 {type=Curve, loc=<0 0 0>, rot=<0 0 0 1>}
 Pose_316 {type=Curve, loc=<0 0 0>, rot=<0 0 0 1>}
 Pose_317 {type=Curve, loc=<0 0 0>, rot=<0 0 0 1>}
 Pose_318 {type=Curve, loc=<0 0 0>, rot=<0 0 0 1>}
 Pose_319 {type=Curve, loc=<0 0 0>, rot=<0 0 0 1>}
 Pose_320 {type=Curve, loc=<0 0 0>, rot=<0 0 0 1>}
 Pose_321 {type=Curve, loc=<0 0 0>, rot=<0 0 0 1>}
 Desert_Island_Jonesy_Mouth {type=Curve, loc=<0 0 0>, rot=<0 0 0 1>}
 Pose_322 {type=Curve, loc=<0 0 0>, rot=<0 0 0 1>}
 HT_L1_TransitionCurve {type=Curve, loc=<0 0 0>, rot=<0 0 0 1>}
 Hair_CapShape {type=Curve, loc=<0 0 0>, rot=<0 0 0 1>}
 Hair_CapShape1 {type=Curve, loc=<0 0 0>, rot=<0 0 0 1>}
 jekyll {type=Curve, loc=<0 0 0>, rot=<0 0 0 1>}
 JH_TransformCurve {type=Curve, loc=<0 0 0>, rot=<0 0 0 1>}
 eyes_LOD0 {type=Curve, loc=<0 0 0>, rot=<0 0 0 1>}
 MaterialCurveA {type=Curve, loc=<0 0 0>, rot=<0 0 0 1>}
 MaterialCurveB {type=Curve, loc=<0 0 0>, rot=<0 0 0 1>}
 MaterialCurveC {type=Curve, loc=<0 0 0>, rot=<0 0 0 1>}
 MaterialCurveD {type=Curve, loc=<0 0 0>, rot=<0 0 0 1>}
 MaterialCurveE {type=Curve, loc=<0 0 0>, rot=<0 0 0 1>}
 Custom_Frontend_Face_Idle {type=Curve, loc=<0 0 0>, rot=<0 0 0 1>}
 blendParent1 {type=Curve, loc=<0 0 0>, rot=<0 0 0 1>}
 BaseHeadDefaultFaceIdle {type=Curve, loc=<0 0 0>, rot=<0 0 0 1>}
 VeiledTransitionGlowAmount {type=Curve, loc=<0 0 0>, rot=<0 0 0 1>}
 is_3L {type=Curve, loc=<0 0 0>, rot=<0 0 0 1>}
 Hair_Hat_LOD0 {type=Curve, loc=<0 0 0>, rot=<0 0 0 1>}
 Hair_Helmet_LOD0 {type=Curve, loc=<0 0 0>, rot=<0 0 0 1>}
 Hair_Null_LOD0 {type=Curve, loc=<0 0 0>, rot=<0 0 0 1>}
 Pose_323 {type=Curve, loc=<0 0 0>, rot=<0 0 0 1>}
 Pose_324 {type=Curve, loc=<0 0 0>, rot=<0 0 0 1>}
 Pose_325 {type=Curve, loc=<0 0 0>, rot=<0 0 0 1>}
 Pose_326 {type=Curve, loc=<0 0 0>, rot=<0 0 0 1>}
 Pose_327 {type=Curve, loc=<0 0 0>, rot=<0 0 0 1>}
 Pose_328 {type=Curve, loc=<0 0 0>, rot=<0 0 0 1>}
 Pose_329 {type=Curve, loc=<0 0 0>, rot=<0 0 0 1>}
 Pose_330 {type=Curve, loc=<0 0 0>, rot=<0 0 0 1>}
 Pose_331 {type=Curve, loc=<0 0 0>, rot=<0 0 0 1>}
 Pose_332 {type=Curve, loc=<0 0 0>, rot=<0 0 0 1>}
 Pose_333 {type=Curve, loc=<0 0 0>, rot=<0 0 0 1>}
 Pose_334 {type=Curve, loc=<0 0 0>, rot=<0 0 0 1>}
 Pose_335 {type=Curve, loc=<0 0 0>, rot=<0 0 0 1>}
 Pose_336 {type=Curve, loc=<0 0 0>, rot=<0 0 0 1>}
 Pose_337 {type=Curve, loc=<0 0 0>, rot=<0 0 0 1>}
 Pose_338 {type=Curve, loc=<0 0 0>, rot=<0 0 0 1>}
 Pose_339 {type=Curve, loc=<0 0 0>, rot=<0 0 0 1>}
 Pose_340 {type=Curve, loc=<0 0 0>, rot=<0 0 0 1>}
 Pose_341 {type=Curve, loc=<0 0 0>, rot=<0 0 0 1>}
 Pose_342 {type=Curve, loc=<0 0 0>, rot=<0 0 0 1>}
 Pose_343 {type=Curve, loc=<0 0 0>, rot=<0 0 0 1>}
 Pose_344 {type=Curve, loc=<0 0 0>, rot=<0 0 0 1>}
 Pose_345 {type=Curve, loc=<0 0 0>, rot=<0 0 0 1>}
 Pose_346 {type=Curve, loc=<0 0 0>, rot=<0 0 0 1>}
 Pose_347 {type=Curve, loc=<0 0 0>, rot=<0 0 0 1>}
 Pose_348 {type=Curve, loc=<0 0 0>, rot=<0 0 0 1>}
 Pose_349 {type=Curve, loc=<0 0 0>, rot=<0 0 0 1>}
 Pose_350 {type=Curve, loc=<0 0 0>, rot=<0 0 0 1>}
 Pose_351 {type=Curve, loc=<0 0 0>, rot=<0 0 0 1>}
 Pose_352 {type=Curve, loc=<0 0 0>, rot=<0 0 0 1>}
 Pose_353 {type=Curve, loc=<0 0 0>, rot=<0 0 0 1>}
 Pose_354 {type=Curve, loc=<0 0 0>, rot=<0 0 0 1>}
 Pose_355 {type=Curve, loc=<0 0 0>, rot=<0 0 0 1>}
 Pose_356 {type=Curve, loc=<0 0 0>, rot=<0 0 0 1>}
 Pose_357 {type=Curve, loc=<0 0 0>, rot=<0 0 0 1>}
 Pose_358 {type=Curve, loc=<0 0 0>, rot=<0 0 0 1>}
 Pose_359 {type=Curve, loc=<0 0 0>, rot=<0 0 0 1>}
 Pose_360 {type=Curve, loc=<0 0 0>, rot=<0 0 0 1>}
 Pose_361 {type=Curve, loc=<0 0 0>, rot=<0 0 0 1>}
 Pose_362 {type=Curve, loc=<0 0 0>, rot=<0 0 0 1>}
 Pose_363 {type=Curve, loc=<0 0 0>, rot=<0 0 0 1>}
 Pose_364 {type=Curve, loc=<0 0 0>, rot=<0 0 0 1>}
 Pose_365 {type=Curve, loc=<0 0 0>, rot=<0 0 0 1>}
 Pose_366 {type=Curve, loc=<0 0 0>, rot=<0 0 0 1>}
 Pose_367 {type=Curve, loc=<0 0 0>, rot=<0 0 0 1>}
 Pose_368 {type=Curve, loc=<0 0 0>, rot=<0 0 0 1>}
 Pose_369 {type=Curve, loc=<0 0 0>, rot=<0 0 0 1>}
 Pose_370 {type=Curve, loc=<0 0 0>, rot=<0 0 0 1>}
 Pose_371 {type=Curve, loc=<0 0 0>, rot=<0 0 0 1>}
 Pose_372 {type=Curve, loc=<0 0 0>, rot=<0 0 0 1>}
 Pose_373 {type=Curve, loc=<0 0 0>, rot=<0 0 0 1>}
 Pose_374 {type=Curve, loc=<0 0 0>, rot=<0 0 0 1>}
 Pose_375 {type=Curve, loc=<0 0 0>, rot=<0 0 0 1>}
 Pose_376 {type=Curve, loc=<0 0 0>, rot=<0 0 0 1>}
 Pose_377 {type=Curve, loc=<0 0 0>, rot=<0 0 0 1>}
 Pose_378 {type=Curve, loc=<0 0 0>, rot=<0 0 0 1>}
 Pose_379 {type=Curve, loc=<0 0 0>, rot=<0 0 0 1>}
 Pose_380 {type=Curve, loc=<0 0 0>, rot=<0 0 0 1>}
 Pose_381 {type=Curve, loc=<0 0 0>, rot=<0 0 0 1>}
 Pose_382 {type=Curve, loc=<0 0 0>, rot=<0 0 0 1>}
 Pose_383 {type=Curve, loc=<0 0 0>, rot=<0 0 0 1>}
 Pose_384 {type=Curve, loc=<0 0 0>, rot=<0 0 0 1>}
 Pose_385 {type=Curve, loc=<0 0 0>, rot=<0 0 0 1>}
 Pose_386 {type=Curve, loc=<0 0 0>, rot=<0 0 0 1>}
 Pose_387 {type=Curve, loc=<0 0 0>, rot=<0 0 0 1>}
 Pose_388 {type=Curve, loc=<0 0 0>, rot=<0 0 0 1>}
 Pose_389 {type=Curve, loc=<0 0 0>, rot=<0 0 0 1>}
 Pose_390 {type=Curve, loc=<0 0 0>, rot=<0 0 0 1>}
 Pose_391 {type=Curve, loc=<0 0 0>, rot=<0 0 0 1>}
 Pose_392 {type=Curve, loc=<0 0 0>, rot=<0 0 0 1>}
 Pose_393 {type=Curve, loc=<0 0 0>, rot=<0 0 0 1>}
 Pose_394 {type=Curve, loc=<0 0 0>, rot=<0 0 0 1>}
 Pose_395 {type=Curve, loc=<0 0 0>, rot=<0 0 0 1>}
 Pose_396 {type=Curve, loc=<0 0 0>, rot=<0 0 0 1>}
 Pose_397 {type=Curve, loc=<0 0 0>, rot=<0 0 0 1>}
 Pose_398 {type=Curve, loc=<0 0 0>, rot=<0 0 0 1>}
 Pose_399 {type=Curve, loc=<0 0 0>, rot=<0 0 0 1>}
 Pose_400 {type=Curve, loc=<0 0 0>, rot=<0 0 0 1>}
 Pose_401 {type=Curve, loc=<0 0 0>, rot=<0 0 0 1>}
 Pose_402 {type=Curve, loc=<0 0 0>, rot=<0 0 0 1>}
 Pose_403 {type=Curve, loc=<0 0 0>, rot=<0 0 0 1>}
 Pose_404 {type=Curve, loc=<0 0 0>, rot=<0 0 0 1>}
 Pose_405 {type=Curve, loc=<0 0 0>, rot=<0 0 0 1>}
 Pose_406 {type=Curve, loc=<0 0 0>, rot=<0 0 0 1>}
 Pose_407 {type=Curve, loc=<0 0 0>, rot=<0 0 0 1>}
 Pose_408 {type=Curve, loc=<0 0 0>, rot=<0 0 0 1>}
 Pose_409 {type=Curve, loc=<0 0 0>, rot=<0 0 0 1>}
 Pose_410 {type=Curve, loc=<0 0 0>, rot=<0 0 0 1>}
 Pose_411 {type=Curve, loc=<0 0 0>, rot=<0 0 0 1>}
 Pose_412 {type=Curve, loc=<0 0 0>, rot=<0 0 0 1>}
 Pose_413 {type=Curve, loc=<0 0 0>, rot=<0 0 0 1>}
 Pose_414 {type=Curve, loc=<0 0 0>, rot=<0 0 0 1>}
 mouth_close_pose {type=Curve, loc=<0 0 0>, rot=<0 0 0 1>}
 mouth_pucker_pose {type=Curve, loc=<0 0 0>, rot=<0 0 0 1>}
 mouth_left_pose {type=Curve, loc=<0 0 0>, rot=<0 0 0 1>}
 mouth_right_pose {type=Curve, loc=<0 0 0>, rot=<0 0 0 1>}
 L_mouth_stretch_pose {type=Curve, loc=<0 0 0>, rot=<0 0 0 1>}
 R_mouth_stretch_pose {type=Curve, loc=<0 0 0>, rot=<0 0 0 1>}
 mouth_roll_lower_pose {type=Curve, loc=<0 0 0>, rot=<0 0 0 1>}
 mouth_roll_upper_pose {type=Curve, loc=<0 0 0>, rot=<0 0 0 1>}
 mouth_shrug_lower_pose {type=Curve, loc=<0 0 0>, rot=<0 0 0 1>}
 mouth_shrug_upper_pose {type=Curve, loc=<0 0 0>, rot=<0 0 0 1>}
 L_mouth_press_pose {type=Curve, loc=<0 0 0>, rot=<0 0 0 1>}
 R_mouth_press_pose {type=Curve, loc=<0 0 0>, rot=<0 0 0 1>}
 cheek_puff_pose {type=Curve, loc=<0 0 0>, rot=<0 0 0 1>}
 L_nose_wrinkler_pose {type=Curve, loc=<0 0 0>, rot=<0 0 0 1>}
 R_nose_wrinkler_pose {type=Curve, loc=<0 0 0>, rot=<0 0 0 1>}
 Pose_415 {type=Curve, loc=<0 0 0>, rot=<0 0 0 1>}
 Pose_416 {type=Curve, loc=<0 0 0>, rot=<0 0 0 1>}
 Pose_417 {type=Curve, loc=<0 0 0>, rot=<0 0 0 1>}
 Pose_418 {type=Curve, loc=<0 0 0>, rot=<0 0 0 1>}
 Pose_419 {type=Curve, loc=<0 0 0>, rot=<0 0 0 1>}
 Pose_420 {type=Curve, loc=<0 0 0>, rot=<0 0 0 1>}
 Pose_421 {type=Curve, loc=<0 0 0>, rot=<0 0 0 1>}
 Pose_422 {type=Curve, loc=<0 0 0>, rot=<0 0 0 1>}
 Pose_423 {type=Curve, loc=<0 0 0>, rot=<0 0 0 1>}
 Pose_424 {type=Curve, loc=<0 0 0>, rot=<0 0 0 1>}
 Pose_425 {type=Curve, loc=<0 0 0>, rot=<0 0 0 1>}
 Pose_426 {type=Curve, loc=<0 0 0>, rot=<0 0 0 1>}
 Pose_427 {type=Curve, loc=<0 0 0>, rot=<0 0 0 1>}
 Pose_428 {type=Curve, loc=<0 0 0>, rot=<0 0 0 1>}
 Pose_429 {type=Curve, loc=<0 0 0>, rot=<0 0 0 1>}
 Pose_430 {type=Curve, loc=<0 0 0>, rot=<0 0 0 1>}
 Pose_431 {type=Curve, loc=<0 0 0>, rot=<0 0 0 1>}
 Pose_432 {type=Curve, loc=<0 0 0>, rot=<0 0 0 1>}
 Pose_433 {type=Curve, loc=<0 0 0>, rot=<0 0 0 1>}
 Pose_434 {type=Curve, loc=<0 0 0>, rot=<0 0 0 1>}
 Pose_435 {type=Curve, loc=<0 0 0>, rot=<0 0 0 1>}
 Pose_436 {type=Curve, loc=<0 0 0>, rot=<0 0 0 1>}
 Pose_437 {type=Curve, loc=<0 0 0>, rot=<0 0 0 1>}
 Pose_438 {type=Curve, loc=<0 0 0>, rot=<0 0 0 1>}
 Pose_439 {type=Curve, loc=<0 0 0>, rot=<0 0 0 1>}
 Pose_440 {type=Curve, loc=<0 0 0>, rot=<0 0 0 1>}
 Pose_441 {type=Curve, loc=<0 0 0>, rot=<0 0 0 1>}
 Pose_442 {type=Curve, loc=<0 0 0>, rot=<0 0 0 1>}
 Pose_443 {type=Curve, loc=<0 0 0>, rot=<0 0 0 1>}
 Pose_444 {type=Curve, loc=<0 0 0>, rot=<0 0 0 1>}
 Pose_445 {type=Curve, loc=<0 0 0>, rot=<0 0 0 1>}
 Pose_446 {type=Curve, loc=<0 0 0>, rot=<0 0 0 1>}
 Pose_447 {type=Curve, loc=<0 0 0>, rot=<0 0 0 1>}
 Pose_448 {type=Curve, loc=<0 0 0>, rot=<0 0 0 1>}
 Pose_449 {type=Curve, loc=<0 0 0>, rot=<0 0 0 1>}
 Pose_450 {type=Curve, loc=<0 0 0>, rot=<0 0 0 1>}
 Pose_451 {type=Curve, loc=<0 0 0>, rot=<0 0 0 1>}
 Pose_452 {type=Curve, loc=<0 0 0>, rot=<0 0 0 1>}
 Pose_453 {type=Curve, loc=<0 0 0>, rot=<0 0 0 1>}
 Pose_454 {type=Curve, loc=<0 0 0>, rot=<0 0 0 1>}
 Pose_455 {type=Curve, loc=<0 0 0>, rot=<0 0 0 1>}
 Pose_456 {type=Curve, loc=<0 0 0>, rot=<0 0 0 1>}
 Pose_457 {type=Curve, loc=<0 0 0>, rot=<0 0 0 1>}
 Pose_458 {type=Curve, loc=<0 0 0>, rot=<0 0 0 1>}
 Pose_459 {type=Curve, loc=<0 0 0>, rot=<0 0 0 1>}
 Pose_460 {type=Curve, loc=<0 0 0>, rot=<0 0 0 1>}
 Pose_461 {type=Curve, loc=<0 0 0>, rot=<0 0 0 1>}
 Pose_462 {type=Curve, loc=<0 0 0>, rot=<0 0 0 1>}
 Pose_463 {type=Curve, loc=<0 0 0>, rot=<0 0 0 1>}
 Pose_464 {type=Curve, loc=<0 0 0>, rot=<0 0 0 1>}
 Pose_465 {type=Curve, loc=<0 0 0>, rot=<0 0 0 1>}
 Pose_466 {type=Curve, loc=<0 0 0>, rot=<0 0 0 1>}
 Pose_467 {type=Curve, loc=<0 0 0>, rot=<0 0 0 1>}
 Pose_468 {type=Curve, loc=<0 0 0>, rot=<0 0 0 1>}
 Pose_469 {type=Curve, loc=<0 0 0>, rot=<0 0 0 1>}
 Pose_470 {type=Curve, loc=<0 0 0>, rot=<0 0 0 1>}
 Pose_471 {type=Curve, loc=<0 0 0>, rot=<0 0 0 1>}
 Pose_472 {type=Curve, loc=<0 0 0>, rot=<0 0 0 1>}
 Pose_473 {type=Curve, loc=<0 0 0>, rot=<0 0 0 1>}
 Pose_474 {type=Curve, loc=<0 0 0>, rot=<0 0 0 1>}
 Pose_475 {type=Curve, loc=<0 0 0>, rot=<0 0 0 1>}
 Pose_476 {type=Curve, loc=<0 0 0>, rot=<0 0 0 1>}
 Pose_477 {type=Curve, loc=<0 0 0>, rot=<0 0 0 1>}
 Pose_478 {type=Curve, loc=<0 0 0>, rot=<0 0 0 1>}
 Pose_479 {type=Curve, loc=<0 0 0>, rot=<0 0 0 1>}
 Pose_480 {type=Curve, loc=<0 0 0>, rot=<0 0 0 1>}
 Pose_481 {type=Curve, loc=<0 0 0>, rot=<0 0 0 1>}
 Pose_482 {type=Curve, loc=<0 0 0>, rot=<0 0 0 1>}
 Pose_483 {type=Curve, loc=<0 0 0>, rot=<0 0 0 1>}
 Pose_484 {type=Curve, loc=<0 0 0>, rot=<0 0 0 1>}
 Pose_485 {type=Curve, loc=<0 0 0>, rot=<0 0 0 1>}
 Pose_486 {type=Curve, loc=<0 0 0>, rot=<0 0 0 1>}
 Pose_487 {type=Curve, loc=<0 0 0>, rot=<0 0 0 1>}
 Pose_488 {type=Curve, loc=<0 0 0>, rot=<0 0 0 1>}
 Pose_489 {type=Curve, loc=<0 0 0>, rot=<0 0 0 1>}
 Pose_490 {type=Curve, loc=<0 0 0>, rot=<0 0 0 1>}
 Pose_491 {type=Curve, loc=<0 0 0>, rot=<0 0 0 1>}
 Pose_492 {type=Curve, loc=<0 0 0>, rot=<0 0 0 1>}
 Pose_493 {type=Curve, loc=<0 0 0>, rot=<0 0 0 1>}
 Pose_494 {type=Curve, loc=<0 0 0>, rot=<0 0 0 1>}
 Pose_495 {type=Curve, loc=<0 0 0>, rot=<0 0 0 1>}
 Pose_496 {type=Curve, loc=<0 0 0>, rot=<0 0 0 1>}
 Pose_497 {type=Curve, loc=<0 0 0>, rot=<0 0 0 1>}
 Pose_498 {type=Curve, loc=<0 0 0>, rot=<0 0 0 1>}
 Pose_499 {type=Curve, loc=<0 0 0>, rot=<0 0 0 1>}
 Pose_500 {type=Curve, loc=<0 0 0>, rot=<0 0 0 1>}
 Pose_501 {type=Curve, loc=<0 0 0>, rot=<0 0 0 1>}
 Pose_502 {type=Curve, loc=<0 0 0>, rot=<0 0 0 1>}
 Pose_503 {type=Curve, loc=<0 0 0>, rot=<0 0 0 1>}
 Pose_504 {type=Curve, loc=<0 0 0>, rot=<0 0 0 1>}
 Pose_505 {type=Curve, loc=<0 0 0>, rot=<0 0 0 1>}
 Pose_506 {type=Curve, loc=<0 0 0>, rot=<0 0 0 1>}
 Pose_507 {type=Curve, loc=<0 0 0>, rot=<0 0 0 1>}
 Pose_508 {type=Curve, loc=<0 0 0>, rot=<0 0 0 1>}
 Pose_509 {type=Curve, loc=<0 0 0>, rot=<0 0 0 1>}
 Pose_510 {type=Curve, loc=<0 0 0>, rot=<0 0 0 1>}
 Pose_511 {type=Curve, loc=<0 0 0>, rot=<0 0 0 1>}
 Pose_512 {type=Curve, loc=<0 0 0>, rot=<0 0 0 1>}
 Pose_513 {type=Curve, loc=<0 0 0>, rot=<0 0 0 1>}
 Pose_514 {type=Curve, loc=<0 0 0>, rot=<0 0 0 1>}
 Pose_515 {type=Curve, loc=<0 0 0>, rot=<0 0 0 1>}
 Pose_516 {type=Curve, loc=<0 0 0>, rot=<0 0 0 1>}
 Pose_517 {type=Curve, loc=<0 0 0>, rot=<0 0 0 1>}
 Pose_518 {type=Curve, loc=<0 0 0>, rot=<0 0 0 1>}
 Pose_519 {type=Curve, loc=<0 0 0>, rot=<0 0 0 1>}
 Pose_520 {type=Curve, loc=<0 0 0>, rot=<0 0 0 1>}
 Pose_521 {type=Curve, loc=<0 0 0>, rot=<0 0 0 1>}
 Pose_522 {type=Curve, loc=<0 0 0>, rot=<0 0 0 1>}
 Pose_523 {type=Curve, loc=<0 0 0>, rot=<0 0 0 1>}
 Pose_524 {type=Curve, loc=<0 0 0>, rot=<0 0 0 1>}
 Pose_525 {type=Curve, loc=<0 0 0>, rot=<0 0 0 1>}
 Pose_526 {type=Curve, loc=<0 0 0>, rot=<0 0 0 1>}
 Pose_527 {type=Curve, loc=<0 0 0>, rot=<0 0 0 1>}
 Pose_528 {type=Curve, loc=<0 0 0>, rot=<0 0 0 1>}
 Pose_529 {type=Curve, loc=<0 0 0>, rot=<0 0 0 1>}
 Pose_530 {type=Curve, loc=<0 0 0>, rot=<0 0 0 1>}
 Pose_531 {type=Curve, loc=<0 0 0>, rot=<0 0 0 1>}
 Pose_532 {type=Curve, loc=<0 0 0>, rot=<0 0 0 1>}
 Pose_533 {type=Curve, loc=<0 0 0>, rot=<0 0 0 1>}
 Pose_534 {type=Curve, loc=<0 0 0>, rot=<0 0 0 1>}
 Pose_535 {type=Curve, loc=<0 0 0>, rot=<0 0 0 1>}
 Pose_536 {type=Curve, loc=<0 0 0>, rot=<0 0 0 1>}
 Pose_537 {type=Curve, loc=<0 0 0>, rot=<0 0 0 1>}
 Pose_538 {type=Curve, loc=<0 0 0>, rot=<0 0 0 1>}
 Pose_539 {type=Curve, loc=<0 0 0>, rot=<0 0 0 1>}
 Pose_540 {type=Curve, loc=<0 0 0>, rot=<0 0 0 1>}
 Pose_541 {type=Curve, loc=<0 0 0>, rot=<0 0 0 1>}
 Pose_542 {type=Curve, loc=<0 0 0>, rot=<0 0 0 1>}
 Pose_543 {type=Curve, loc=<0 0 0>, rot=<0 0 0 1>}
 Pose_544 {type=Curve, loc=<0 0 0>, rot=<0 0 0 1>}
 Pose_545 {type=Curve, loc=<0 0 0>, rot=<0 0 0 1>}
 Pose_546 {type=Curve, loc=<0 0 0>, rot=<0 0 0 1>}
 Pose_547 {type=Curve, loc=<0 0 0>, rot=<0 0 0 1>}
 Pose_548 {type=Curve, loc=<0 0 0>, rot=<0 0 0 1>}
 Pose_549 {type=Curve, loc=<0 0 0>, rot=<0 0 0 1>}
 Pose_550 {type=Curve, loc=<0 0 0>, rot=<0 0 0 1>}
 Pose_551 {type=Curve, loc=<0 0 0>, rot=<0 0 0 1>}
 Pose_552 {type=Curve, loc=<0 0 0>, rot=<0 0 0 1>}
 Pose_553 {type=Curve, loc=<0 0 0>, rot=<0 0 0 1>}
 Pose_554 {type=Curve, loc=<0 0 0>, rot=<0 0 0 1>}
 Pose_555 {type=Curve, loc=<0 0 0>, rot=<0 0 0 1>}
 Pose_556 {type=Curve, loc=<0 0 0>, rot=<0 0 0 1>}
 Pose_557 {type=Curve, loc=<0 0 0>, rot=<0 0 0 1>}
 Pose_558 {type=Curve, loc=<0 0 0>, rot=<0 0 0 1>}
 Pose_559 {type=Curve, loc=<0 0 0>, rot=<0 0 0 1>}
 Pose_560 {type=Curve, loc=<0 0 0>, rot=<0 0 0 1>}
 Pose_561 {type=Curve, loc=<0 0 0>, rot=<0 0 0 1>}
 Pose_562 {type=Curve, loc=<0 0 0>, rot=<0 0 0 1>}
 Pose_563 {type=Curve, loc=<0 0 0>, rot=<0 0 0 1>}
 Pose_564 {type=Curve, loc=<0 0 0>, rot=<0 0 0 1>}
 Pose_565 {type=Curve, loc=<0 0 0>, rot=<0 0 0 1>}
 Pose_566 {type=Curve, loc=<0 0 0>, rot=<0 0 0 1>}
 Pose_567 {type=Curve, loc=<0 0 0>, rot=<0 0 0 1>}
 Pose_568 {type=Curve, loc=<0 0 0>, rot=<0 0 0 1>}
 Pose_569 {type=Curve, loc=<0 0 0>, rot=<0 0 0 1>}
 Pose_570 {type=Curve, loc=<0 0 0>, rot=<0 0 0 1>}
 Pose_571 {type=Curve, loc=<0 0 0>, rot=<0 0 0 1>}
 Pose_572 {type=Curve, loc=<0 0 0>, rot=<0 0 0 1>}
 Pose_573 {type=Curve, loc=<0 0 0>, rot=<0 0 0 1>}
 Pose_574 {type=Curve, loc=<0 0 0>, rot=<0 0 0 1>}
 Pose_575 {type=Curve, loc=<0 0 0>, rot=<0 0 0 1>}
 Pose_576 {type=Curve, loc=<0 0 0>, rot=<0 0 0 1>}
 Pose_577 {type=Curve, loc=<0 0 0>, rot=<0 0 0 1>}
 Pose_578 {type=Curve, loc=<0 0 0>, rot=<0 0 0 1>}
 Pose_579 {type=Curve, loc=<0 0 0>, rot=<0 0 0 1>}
 Pose_580 {type=Curve, loc=<0 0 0>, rot=<0 0 0 1>}
 Pose_581 {type=Curve, loc=<0 0 0>, rot=<0 0 0 1>}
 Pose_582 {type=Curve, loc=<0 0 0>, rot=<0 0 0 1>}
 Pose_583 {type=Curve, loc=<0 0 0>, rot=<0 0 0 1>}
 Pose_584 {type=Curve, loc=<0 0 0>, rot=<0 0 0 1>}
 Pose_585 {type=Curve, loc=<0 0 0>, rot=<0 0 0 1>}
 Pose_586 {type=Curve, loc=<0 0 0>, rot=<0 0 0 1>}
 Pose_587 {type=Curve, loc=<0 0 0>, rot=<0 0 0 1>}
 Pose_588 {type=Curve, loc=<0 0 0>, rot=<0 0 0 1>}
 Pose_589 {type=Curve, loc=<0 0 0>, rot=<0 0 0 1>}
 Pose_590 {type=Curve, loc=<0 0 0>, rot=<0 0 0 1>}
 Pose_591 {type=Curve, loc=<0 0 0>, rot=<0 0 0 1>}
 Pose_592 {type=Curve, loc=<0 0 0>, rot=<0 0 0 1>}
 Pose_593 {type=Curve, loc=<0 0 0>, rot=<0 0 0 1>}
 Pose_594 {type=Curve, loc=<0 0 0>, rot=<0 0 0 1>}
 Pose_595 {type=Curve, loc=<0 0 0>, rot=<0 0 0 1>}
 Pose_596 {type=Curve, loc=<0 0 0>, rot=<0 0 0 1>}
 Pose_597 {type=Curve, loc=<0 0 0>, rot=<0 0 0 1>}
 Pose_598 {type=Curve, loc=<0 0 0>, rot=<0 0 0 1>}
 Pose_599 {type=Curve, loc=<0 0 0>, rot=<0 0 0 1>}
 Pose_600 {type=Curve, loc=<0 0 0>, rot=<0 0 0 1>}
 Pose_601 {type=Curve, loc=<0 0 0>, rot=<0 0 0 1>}
 Pose_602 {type=Curve, loc=<0 0 0>, rot=<0 0 0 1>}
 Pose_603 {type=Curve, loc=<0 0 0>, rot=<0 0 0 1>}
 Pose_604 {type=Curve, loc=<0 0 0>, rot=<0 0 0 1>}
 Pose_605 {type=Curve, loc=<0 0 0>, rot=<0 0 0 1>}
 Pose_606 {type=Curve, loc=<0 0 0>, rot=<0 0 0 1>}
 Pose_607 {type=Curve, loc=<0 0 0>, rot=<0 0 0 1>}
 Pose_608 {type=Curve, loc=<0 0 0>, rot=<0 0 0 1>}
 Pose_609 {type=Curve, loc=<0 0 0>, rot=<0 0 0 1>}
 Pose_610 {type=Curve, loc=<0 0 0>, rot=<0 0 0 1>}
 Pose_611 {type=Curve, loc=<0 0 0>, rot=<0 0 0 1>}
 Pose_612 {type=Curve, loc=<0 0 0>, rot=<0 0 0 1>}
 Pose_613 {type=Curve, loc=<0 0 0>, rot=<0 0 0 1>}
 Pose_614 {type=Curve, loc=<0 0 0>, rot=<0 0 0 1>}
 Pose_615 {type=Curve, loc=<0 0 0>, rot=<0 0 0 1>}
 Pose_616 {type=Curve, loc=<0 0 0>, rot=<0 0 0 1>}
 Pose_617 {type=Curve, loc=<0 0 0>, rot=<0 0 0 1>}
 Pose_618 {type=Curve, loc=<0 0 0>, rot=<0 0 0 1>}
 Pose_619 {type=Curve, loc=<0 0 0>, rot=<0 0 0 1>}
 Pose_620 {type=Curve, loc=<0 0 0>, rot=<0 0 0 1>}
 Pose_621 {type=Curve, loc=<0 0 0>, rot=<0 0 0 1>}
 Pose_622 {type=Curve, loc=<0 0 0>, rot=<0 0 0 1>}
 Pose_623 {type=Curve, loc=<0 0 0>, rot=<0 0 0 1>}
 Pose_624 {type=Curve, loc=<0 0 0>, rot=<0 0 0 1>}
 Pose_625 {type=Curve, loc=<0 0 0>, rot=<0 0 0 1>}
 Pose_626 {type=Curve, loc=<0 0 0>, rot=<0 0 0 1>}
 Pose_627 {type=Curve, loc=<0 0 0>, rot=<0 0 0 1>}
 Pose_628 {type=Curve, loc=<0 0 0>, rot=<0 0 0 1>}
 Pose_629 {type=Curve, loc=<0 0 0>, rot=<0 0 0 1>}
 Pose_630 {type=Curve, loc=<0 0 0>, rot=<0 0 0 1>}
 Pose_631 {type=Curve, loc=<0 0 0>, rot=<0 0 0 1>}
 Pose_632 {type=Curve, loc=<0 0 0>, rot=<0 0 0 1>}
 Pose_633 {type=Curve, loc=<0 0 0>, rot=<0 0 0 1>}
 Pose_634 {type=Curve, loc=<0 0 0>, rot=<0 0 0 1>}
 Pose_635 {type=Curve, loc=<0 0 0>, rot=<0 0 0 1>}
 Pose_636 {type=Curve, loc=<0 0 0>, rot=<0 0 0 1>}
 Pose_637 {type=Curve, loc=<0 0 0>, rot=<0 0 0 1>}
 Pose_638 {type=Curve, loc=<0 0 0>, rot=<0 0 0 1>}
 Pose_639 {type=Curve, loc=<0 0 0>, rot=<0 0 0 1>}
 Pose_640 {type=Curve, loc=<0 0 0>, rot=<0 0 0 1>}
 Pose_641 {type=Curve, loc=<0 0 0>, rot=<0 0 0 1>}
 Pose_642 {type=Curve, loc=<0 0 0>, rot=<0 0 0 1>}
 Pose_643 {type=Curve, loc=<0 0 0>, rot=<0 0 0 1>}
 Pose_644 {type=Curve, loc=<0 0 0>, rot=<0 0 0 1>}
 Pose_645 {type=Curve, loc=<0 0 0>, rot=<0 0 0 1>}
 Pose_646 {type=Curve, loc=<0 0 0>, rot=<0 0 0 1>}
 Pose_647 {type=Curve, loc=<0 0 0>, rot=<0 0 0 1>}
 Pose_648 {type=Curve, loc=<0 0 0>, rot=<0 0 0 1>}
 Pose_649 {type=Curve, loc=<0 0 0>, rot=<0 0 0 1>}
 Pose_650 {type=Curve, loc=<0 0 0>, rot=<0 0 0 1>}
 Pose_651 {type=Curve, loc=<0 0 0>, rot=<0 0 0 1>}
 base {type=Curve, loc=<0 0 0>, rot=<0 0 0 1>}
 eyeLookDownLeft {type=Curve, loc=<0 0 0>, rot=<0 0 0 1>}
 eyeLookInLeft {type=Curve, loc=<0 0 0>, rot=<0 0 0 1>}
 eyeLookUpLeft {type=Curve, loc=<0 0 0>, rot=<0 0 0 1>}
 eyeLookInRight {type=Curve, loc=<0 0 0>, rot=<0 0 0 1>}
 Pose_652 {type=Curve, loc=<0 0 0>, rot=<0 0 0 1>}
 Pose_653 {type=Curve, loc=<0 0 0>, rot=<0 0 0 1>}
 Pose_654 {type=Curve, loc=<0 0 0>, rot=<0 0 0 1>}
 Pose_655 {type=Curve, loc=<0 0 0>, rot=<0 0 0 1>}
 Pose_656 {type=Curve, loc=<0 0 0>, rot=<0 0 0 1>}
 Pose_657 {type=Curve, loc=<0 0 0>, rot=<0 0 0 1>}
 Pose_658 {type=Curve, loc=<0 0 0>, rot=<0 0 0 1>}
 Pose_659 {type=Curve, loc=<0 0 0>, rot=<0 0 0 1>}
 Pose_660 {type=Curve, loc=<0 0 0>, rot=<0 0 0 1>}
 Pose_661 {type=Curve, loc=<0 0 0>, rot=<0 0 0 1>}
 Pose_662 {type=Curve, loc=<0 0 0>, rot=<0 0 0 1>}
 Pose_663 {type=Curve, loc=<0 0 0>, rot=<0 0 0 1>}
 Pose_664 {type=Curve, loc=<0 0 0>, rot=<0 0 0 1>}
 Pose_665 {type=Curve, loc=<0 0 0>, rot=<0 0 0 1>}
 Pose_666 {type=Curve, loc=<0 0 0>, rot=<0 0 0 1>}
 Pose_667 {type=Curve, loc=<0 0 0>, rot=<0 0 0 1>}
 Pose_668 {type=Curve, loc=<0 0 0>, rot=<0 0 0 1>}
 Pose_669 {type=Curve, loc=<0 0 0>, rot=<0 0 0 1>}
 Pose_670 {type=Curve, loc=<0 0 0>, rot=<0 0 0 1>}
 Pose_671 {type=Curve, loc=<0 0 0>, rot=<0 0 0 1>}
 Pose_672 {type=Curve, loc=<0 0 0>, rot=<0 0 0 1>}
 Pose_673 {type=Curve, loc=<0 0 0>, rot=<0 0 0 1>}
 Pose_674 {type=Curve, loc=<0 0 0>, rot=<0 0 0 1>}
 Pose_675 {type=Curve, loc=<0 0 0>, rot=<0 0 0 1>}
 Pose_676 {type=Curve, loc=<0 0 0>, rot=<0 0 0 1>}
 Pose_677 {type=Curve, loc=<0 0 0>, rot=<0 0 0 1>}
 Pose_678 {type=Curve, loc=<0 0 0>, rot=<0 0 0 1>}
 Pose_679 {type=Curve, loc=<0 0 0>, rot=<0 0 0 1>}
 Pose_680 {type=Curve, loc=<0 0 0>, rot=<0 0 0 1>}
 Pose_681 {type=Curve, loc=<0 0 0>, rot=<0 0 0 1>}
 Pose_682 {type=Curve, loc=<0 0 0>, rot=<0 0 0 1>}
 Pose_683 {type=Curve, loc=<0 0 0>, rot=<0 0 0 1>}
 Pose_684 {type=Curve, loc=<0 0 0>, rot=<0 0 0 1>}
 Pose_685 {type=Curve, loc=<0 0 0>, rot=<0 0 0 1>}
 Pose_686 {type=Curve, loc=<0 0 0>, rot=<0 0 0 1>}
 Pose_687 {type=Curve, loc=<0 0 0>, rot=<0 0 0 1>}
 Pose_688 {type=Curve, loc=<0 0 0>, rot=<0 0 0 1>}
 Pose_689 {type=Curve, loc=<0 0 0>, rot=<0 0 0 1>}
 Pose_690 {type=Curve, loc=<0 0 0>, rot=<0 0 0 1>}
 Pose_691 {type=Curve, loc=<0 0 0>, rot=<0 0 0 1>}
 Pose_692 {type=Curve, loc=<0 0 0>, rot=<0 0 0 1>}
 Pose_693 {type=Curve, loc=<0 0 0>, rot=<0 0 0 1>}
 Pose_694 {type=Curve, loc=<0 0 0>, rot=<0 0 0 1>}
 Pose_695 {type=Curve, loc=<0 0 0>, rot=<0 0 0 1>}
 Pose_696 {type=Curve, loc=<0 0 0>, rot=<0 0 0 1>}
 Pose_697 {type=Curve, loc=<0 0 0>, rot=<0 0 0 1>}
 Pose_698 {type=Curve, loc=<0 0 0>, rot=<0 0 0 1>}
 Pose_699 {type=Curve, loc=<0 0 0>, rot=<0 0 0 1>}
 Pose_700 {type=Curve, loc=<0 0 0>, rot=<0 0 0 1>}
 Pose_701 {type=Curve, loc=<0 0 0>, rot=<0 0 0 1>}
 Pose_702 {type=Curve, loc=<0 0 0>, rot=<0 0 0 1>}
 Pose_703 {type=Curve, loc=<0 0 0>, rot=<0 0 0 1>}
 Pose_704 {type=Curve, loc=<0 0 0>, rot=<0 0 0 1>}
 Pose_705 {type=Curve, loc=<0 0 0>, rot=<0 0 0 1>}
 Pose_706 {type=Curve, loc=<0 0 0>, rot=<0 0 0 1>}
 Pose_707 {type=Curve, loc=<0 0 0>, rot=<0 0 0 1>}
 Pose_708 {type=Curve, loc=<0 0 0>, rot=<0 0 0 1>}
 Pose_709 {type=Curve, loc=<0 0 0>, rot=<0 0 0 1>}
 Pose_710 {type=Curve, loc=<0 0 0>, rot=<0 0 0 1>}
 Pose_711 {type=Curve, loc=<0 0 0>, rot=<0 0 0 1>}
 Pose_712 {type=Curve, loc=<0 0 0>, rot=<0 0 0 1>}
 Pose_713 {type=Curve, loc=<0 0 0>, rot=<0 0 0 1>}
 Pose_714 {type=Curve, loc=<0 0 0>, rot=<0 0 0 1>}
 Pose_715 {type=Curve, loc=<0 0 0>, rot=<0 0 0 1>}
 Pose_716 {type=Curve, loc=<0 0 0>, rot=<0 0 0 1>}
 Pose_717 {type=Curve, loc=<0 0 0>, rot=<0 0 0 1>}
 Pose_718 {type=Curve, loc=<0 0 0>, rot=<0 0 0 1>}
 Pose_719 {type=Curve, loc=<0 0 0>, rot=<0 0 0 1>}
 Pose_720 {type=Curve, loc=<0 0 0>, rot=<0 0 0 1>}
 Pose_721 {type=Curve, loc=<0 0 0>, rot=<0 0 0 1>}
 Pose_722 {type=Curve, loc=<0 0 0>, rot=<0 0 0 1>}
 Pose_723 {type=Curve, loc=<0 0 0>, rot=<0 0 0 1>}
 Pose_724 {type=Curve, loc=<0 0 0>, rot=<0 0 0 1>}
 Pose_725 {type=Curve, loc=<0 0 0>, rot=<0 0 0 1>}
 Pose_726 {type=Curve, loc=<0 0 0>, rot=<0 0 0 1>}
 Pose_727 {type=Curve, loc=<0 0 0>, rot=<0 0 0 1>}
 Pose_728 {type=Curve, loc=<0 0 0>, rot=<0 0 0 1>}
 Pose_729 {type=Curve, loc=<0 0 0>, rot=<0 0 0 1>}
 Pose_730 {type=Curve, loc=<0 0 0>, rot=<0 0 0 1>}
 Pose_731 {type=Curve, loc=<0 0 0>, rot=<0 0 0 1>}
 Pose_732 {type=Curve, loc=<0 0 0>, rot=<0 0 0 1>}
 Pose_733 {type=Curve, loc=<0 0 0>, rot=<0 0 0 1>}
 Pose_734 {type=Curve, loc=<0 0 0>, rot=<0 0 0 1>}
 Pose_735 {type=Curve, loc=<0 0 0>, rot=<0 0 0 1>}
 Pose_736 {type=Curve, loc=<0 0 0>, rot=<0 0 0 1>}
 Pose_737 {type=Curve, loc=<0 0 0>, rot=<0 0 0 1>}
 Pose_738 {type=Curve, loc=<0 0 0>, rot=<0 0 0 1>}
 Pose_739 {type=Curve, loc=<0 0 0>, rot=<0 0 0 1>}
 Pose_740 {type=Curve, loc=<0 0 0>, rot=<0 0 0 1>}
 Pose_741 {type=Curve, loc=<0 0 0>, rot=<0 0 0 1>}
 Pose_742 {type=Curve, loc=<0 0 0>, rot=<0 0 0 1>}
 Pose_743 {type=Curve, loc=<0 0 0>, rot=<0 0 0 1>}
 Pose_744 {type=Curve, loc=<0 0 0>, rot=<0 0 0 1>}
 Pose_745 {type=Curve, loc=<0 0 0>, rot=<0 0 0 1>}
 Pose_746 {type=Curve, loc=<0 0 0>, rot=<0 0 0 1>}
 Pose_747 {type=Curve, loc=<0 0 0>, rot=<0 0 0 1>}
 Pose_748 {type=Curve, loc=<0 0 0>, rot=<0 0 0 1>}
 Pose_749 {type=Curve, loc=<0 0 0>, rot=<0 0 0 1>}
 Pose_750 {type=Curve, loc=<0 0 0>, rot=<0 0 0 1>}
 Pose_751 {type=Curve, loc=<0 0 0>, rot=<0 0 0 1>}
 Pose_752 {type=Curve, loc=<0 0 0>, rot=<0 0 0 1>}
 Pose_753 {type=Curve, loc=<0 0 0>, rot=<0 0 0 1>}
 Pose_754 {type=Curve, loc=<0 0 0>, rot=<0 0 0 1>}
 Pose_755 {type=Curve, loc=<0 0 0>, rot=<0 0 0 1>}
 Pose_756 {type=Curve, loc=<0 0 0>, rot=<0 0 0 1>}
 Pose_757 {type=Curve, loc=<0 0 0>, rot=<0 0 0 1>}
 Pose_758 {type=Curve, loc=<0 0 0>, rot=<0 0 0 1>}
 Pose_759 {type=Curve, loc=<0 0 0>, rot=<0 0 0 1>}
 Pose_760 {type=Curve, loc=<0 0 0>, rot=<0 0 0 1>}
 Pose_761 {type=Curve, loc=<0 0 0>, rot=<0 0 0 1>}
 Pose_762 {type=Curve, loc=<0 0 0>, rot=<0 0 0 1>}
 Pose_763 {type=Curve, loc=<0 0 0>, rot=<0 0 0 1>}
 Pose_764 {type=Curve, loc=<0 0 0>, rot=<0 0 0 1>}
 Pose_765 {type=Curve, loc=<0 0 0>, rot=<0 0 0 1>}
 Pose_766 {type=Curve, loc=<0 0 0>, rot=<0 0 0 1>}
 Pose_767 {type=Curve, loc=<0 0 0>, rot=<0 0 0 1>}
 Pose_768 {type=Curve, loc=<0 0 0>, rot=<0 0 0 1>}
 Pose_769 {type=Curve, loc=<0 0 0>, rot=<0 0 0 1>}
 Pose_770 {type=Curve, loc=<0 0 0>, rot=<0 0 0 1>}
 Pose_771 {type=Curve, loc=<0 0 0>, rot=<0 0 0 1>}
 Pose_772 {type=Curve, loc=<0 0 0>, rot=<0 0 0 1>}
 Pose_773 {type=Curve, loc=<0 0 0>, rot=<0 0 0 1>}
 Pose_774 {type=Curve, loc=<0 0 0>, rot=<0 0 0 1>}
 Pose_775 {type=Curve, loc=<0 0 0>, rot=<0 0 0 1>}
 Pose_776 {type=Curve, loc=<0 0 0>, rot=<0 0 0 1>}
 Pose_777 {type=Curve, loc=<0 0 0>, rot=<0 0 0 1>}
 Pose_778 {type=Curve, loc=<0 0 0>, rot=<0 0 0 1>}
 Pose_779 {type=Curve, loc=<0 0 0>, rot=<0 0 0 1>}
 Pose_780 {type=Curve, loc=<0 0 0>, rot=<0 0 0 1>}
 Pose_781 {type=Curve, loc=<0 0 0>, rot=<0 0 0 1>}
 Pose_782 {type=Curve, loc=<0 0 0>, rot=<0 0 0 1>}
 Pose_783 {type=Curve, loc=<0 0 0>, rot=<0 0 0 1>}
 Pose_784 {type=Curve, loc=<0 0 0>, rot=<0 0 0 1>}
 Pose_785 {type=Curve, loc=<0 0 0>, rot=<0 0 0 1>}
 Pose_786 {type=Curve, loc=<0 0 0>, rot=<0 0 0 1>}
 Pose_787 {type=Curve, loc=<0 0 0>, rot=<0 0 0 1>}
 Pose_788 {type=Curve, loc=<0 0 0>, rot=<0 0 0 1>}
 Pose_789 {type=Curve, loc=<0 0 0>, rot=<0 0 0 1>}
 Pose_790 {type=Curve, loc=<0 0 0>, rot=<0 0 0 1>}
 Pose_791 {type=Curve, loc=<0 0 0>, rot=<0 0 0 1>}
 Pose_792 {type=Curve, loc=<0 0 0>, rot=<0 0 0 1>}
 Pose_793 {type=Curve, loc=<0 0 0>, rot=<0 0 0 1>}
 Pose_794 {type=Curve, loc=<0 0 0>, rot=<0 0 0 1>}
 Pose_795 {type=Curve, loc=<0 0 0>, rot=<0 0 0 1>}
 Pose_796 {type=Curve, loc=<0 0 0>, rot=<0 0 0 1>}
 Pose_797 {type=Curve, loc=<0 0 0>, rot=<0 0 0 1>}
 Pose_798 {type=Curve, loc=<0 0 0>, rot=<0 0 0 1>}
 Pose_799 {type=Curve, loc=<0 0 0>, rot=<0 0 0 1>}
 Pose_800 {type=Curve, loc=<0 0 0>, rot=<0 0 0 1>}
 Pose_801 {type=Curve, loc=<0 0 0>, rot=<0 0 0 1>}
 Pose_802 {type=Curve, loc=<0 0 0>, rot=<0 0 0 1>}
 Pose_803 {type=Curve, loc=<0 0 0>, rot=<0 0 0 1>}
 Pose_804 {type=Curve, loc=<0 0 0>, rot=<0 0 0 1>}
 Pose_805 {type=Curve, loc=<0 0 0>, rot=<0 0 0 1>}
 Pose_806 {type=Curve, loc=<0 0 0>, rot=<0 0 0 1>}
 Pose_807 {type=Curve, loc=<0 0 0>, rot=<0 0 0 1>}
 Pose_808 {type=Curve, loc=<0 0 0>, rot=<0 0 0 1>}
 Pose_809 {type=Curve, loc=<0 0 0>, rot=<0 0 0 1>}
 Pose_810 {type=Curve, loc=<0 0 0>, rot=<0 0 0 1>}
 Pose_811 {type=Curve, loc=<0 0 0>, rot=<0 0 0 1>}
 Pose_812 {type=Curve, loc=<0 0 0>, rot=<0 0 0 1>}
 Pose_813 {type=Curve, loc=<0 0 0>, rot=<0 0 0 1>}
 Pose_814 {type=Curve, loc=<0 0 0>, rot=<0 0 0 1>}
 Pose_815 {type=Curve, loc=<0 0 0>, rot=<0 0 0 1>}
 Pose_816 {type=Curve, loc=<0 0 0>, rot=<0 0 0 1>}
 Pose_817 {type=Curve, loc=<0 0 0>, rot=<0 0 0 1>}
 Pose_818 {type=Curve, loc=<0 0 0>, rot=<0 0 0 1>}
 Pose_819 {type=Curve, loc=<0 0 0>, rot=<0 0 0 1>}
 Pose_820 {type=Curve, loc=<0 0 0>, rot=<0 0 0 1>}
 Pose_821 {type=Curve, loc=<0 0 0>, rot=<0 0 0 1>}
 Pose_822 {type=Curve, loc=<0 0 0>, rot=<0 0 0 1>}
 Pose_823 {type=Curve, loc=<0 0 0>, rot=<0 0 0 1>}
 Pose_824 {type=Curve, loc=<0 0 0>, rot=<0 0 0 1>}
 Pose_825 {type=Curve, loc=<0 0 0>, rot=<0 0 0 1>}
 Pose_826 {type=Curve, loc=<0 0 0>, rot=<0 0 0 1>}
 Pose_827 {type=Curve, loc=<0 0 0>, rot=<0 0 0 1>}
 Pose_828 {type=Curve, loc=<0 0 0>, rot=<0 0 0 1>}
 Pose_829 {type=Curve, loc=<0 0 0>, rot=<0 0 0 1>}
 Pose_830 {type=Curve, loc=<0 0 0>, rot=<0 0 0 1>}
 Pose_831 {type=Curve, loc=<0 0 0>, rot=<0 0 0 1>}
 Pose_832 {type=Curve, loc=<0 0 0>, rot=<0 0 0 1>}
 Pose_833 {type=Curve, loc=<0 0 0>, rot=<0 0 0 1>}
 Pose_834 {type=Curve, loc=<0 0 0>, rot=<0 0 0 1>}
 Pose_835 {type=Curve, loc=<0 0 0>, rot=<0 0 0 1>}
 Pose_836 {type=Curve, loc=<0 0 0>, rot=<0 0 0 1>}
 Pose_837 {type=Curve, loc=<0 0 0>, rot=<0 0 0 1>}
 Pose_838 {type=Curve, loc=<0 0 0>, rot=<0 0 0 1>}
 Pose_839 {type=Curve, loc=<0 0 0>, rot=<0 0 0 1>}
 Pose_840 {type=Curve, loc=<0 0 0>, rot=<0 0 0 1>}
 Pose_841 {type=Curve, loc=<0 0 0>, rot=<0 0 0 1>}
 Pose_842 {type=Curve, loc=<0 0 0>, rot=<0 0 0 1>}
 Pose_843 {type=Curve, loc=<0 0 0>, rot=<0 0 0 1>}
 Pose_844 {type=Curve, loc=<0 0 0>, rot=<0 0 0 1>}
 Pose_845 {type=Curve, loc=<0 0 0>, rot=<0 0 0 1>}
 Pose_846 {type=Curve, loc=<0 0 0>, rot=<0 0 0 1>}
 Pose_847 {type=Curve, loc=<0 0 0>, rot=<0 0 0 1>}
 Pose_848 {type=Curve, loc=<0 0 0>, rot=<0 0 0 1>}
 Pose_849 {type=Curve, loc=<0 0 0>, rot=<0 0 0 1>}
 Pose_850 {type=Curve, loc=<0 0 0>, rot=<0 0 0 1>}
 Pose_851 {type=Curve, loc=<0 0 0>, rot=<0 0 0 1>}
 Pose_852 {type=Curve, loc=<0 0 0>, rot=<0 0 0 1>}
 Pose_853 {type=Curve, loc=<0 0 0>, rot=<0 0 0 1>}
 Pose_854 {type=Curve, loc=<0 0 0>, rot=<0 0 0 1>}
 Pose_855 {type=Curve, loc=<0 0 0>, rot=<0 0 0 1>}
 Pose_856 {type=Curve, loc=<0 0 0>, rot=<0 0 0 1>}
 Pose_857 {type=Curve, loc=<0 0 0>, rot=<0 0 0 1>}
 Pose_858 {type=Curve, loc=<0 0 0>, rot=<0 0 0 1>}
 Pose_859 {type=Curve, loc=<0 0 0>, rot=<0 0 0 1>}
 Pose_860 {type=Curve, loc=<0 0 0>, rot=<0 0 0 1>}
 Pose_861 {type=Curve, loc=<0 0 0>, rot=<0 0 0 1>}
 Pose_862 {type=Curve, loc=<0 0 0>, rot=<0 0 0 1>}
 Pose_863 {type=Curve, loc=<0 0 0>, rot=<0 0 0 1>}
 Pose_864 {type=Curve, loc=<0 0 0>, rot=<0 0 0 1>}
 Pose_865 {type=Curve, loc=<0 0 0>, rot=<0 0 0 1>}
 Pose_866 {type=Curve, loc=<0 0 0>, rot=<0 0 0 1>}
 Pose_867 {type=Curve, loc=<0 0 0>, rot=<0 0 0 1>}
 Pose_868 {type=Curve, loc=<0 0 0>, rot=<0 0 0 1>}
 Pose_869 {type=Curve, loc=<0 0 0>, rot=<0 0 0 1>}
 Pose_870 {type=Curve, loc=<0 0 0>, rot=<0 0 0 1>}
 Pose_871 {type=Curve, loc=<0 0 0>, rot=<0 0 0 1>}
 Pose_872 {type=Curve, loc=<0 0 0>, rot=<0 0 0 1>}
 Pose_873 {type=Curve, loc=<0 0 0>, rot=<0 0 0 1>}
 Pose_874 {type=Curve, loc=<0 0 0>, rot=<0 0 0 1>}
 Pose_875 {type=Curve, loc=<0 0 0>, rot=<0 0 0 1>}
 Pose_876 {type=Curve, loc=<0 0 0>, rot=<0 0 0 1>}
 Pose_877 {type=Curve, loc=<0 0 0>, rot=<0 0 0 1>}
 Pose_878 {type=Curve, loc=<0 0 0>, rot=<0 0 0 1>}
 Pose_879 {type=Curve, loc=<0 0 0>, rot=<0 0 0 1>}
 Pose_880 {type=Curve, loc=<0 0 0>, rot=<0 0 0 1>}
 Pose_881 {type=Curve, loc=<0 0 0>, rot=<0 0 0 1>}
 Pose_882 {type=Curve, loc=<0 0 0>, rot=<0 0 0 1>}
 Pose_883 {type=Curve, loc=<0 0 0>, rot=<0 0 0 1>}
 Pose_884 {type=Curve, loc=<0 0 0>, rot=<0 0 0 1>}
 Pose_885 {type=Curve, loc=<0 0 0>, rot=<0 0 0 1>}
 Pose_886 {type=Curve, loc=<0 0 0>, rot=<0 0 0 1>}
 Pose_887 {type=Curve, loc=<0 0 0>, rot=<0 0 0 1>}
 Pose_888 {type=Curve, loc=<0 0 0>, rot=<0 0 0 1>}
 Pose_889 {type=Curve, loc=<0 0 0>, rot=<0 0 0 1>}
 Pose_890 {type=Curve, loc=<0 0 0>, rot=<0 0 0 1>}
 Pose_891 {type=Curve, loc=<0 0 0>, rot=<0 0 0 1>}
 Pose_892 {type=Curve, loc=<0 0 0>, rot=<0 0 0 1>}
 Pose_893 {type=Curve, loc=<0 0 0>, rot=<0 0 0 1>}
 Pose_894 {type=Curve, loc=<0 0 0>, rot=<0 0 0 1>}
 Pose_895 {type=Curve, loc=<0 0 0>, rot=<0 0 0 1>}
 Pose_896 {type=Curve, loc=<0 0 0>, rot=<0 0 0 1>}
 Pose_897 {type=Curve, loc=<0 0 0>, rot=<0 0 0 1>}
 Pose_898 {type=Curve, loc=<0 0 0>, rot=<0 0 0 1>}
 Pose_899 {type=Curve, loc=<0 0 0>, rot=<0 0 0 1>}
 Pose_900 {type=Curve, loc=<0 0 0>, rot=<0 0 0 1>}
 Pose_901 {type=Curve, loc=<0 0 0>, rot=<0 0 0 1>}
 Pose_902 {type=Curve, loc=<0 0 0>, rot=<0 0 0 1>}
 Pose_903 {type=Curve, loc=<0 0 0>, rot=<0 0 0 1>}
 Pose_904 {type=Curve, loc=<0 0 0>, rot=<0 0 0 1>}
 Pose_905 {type=Curve, loc=<0 0 0>, rot=<0 0 0 1>}
 Pose_906 {type=Curve, loc=<0 0 0>, rot=<0 0 0 1>}
 Pose_907 {type=Curve, loc=<0 0 0>, rot=<0 0 0 1>}
 Pose_908 {type=Curve, loc=<0 0 0>, rot=<0 0 0 1>}
 Pose_909 {type=Curve, loc=<0 0 0>, rot=<0 0 0 1>}
 Pose_910 {type=Curve, loc=<0 0 0>, rot=<0 0 0 1>}
 Pose_911 {type=Curve, loc=<0 0 0>, rot=<0 0 0 1>}
 Pose_912 {type=Curve, loc=<0 0 0>, rot=<0 0 0 1>}
 Pose_913 {type=Curve, loc=<0 0 0>, rot=<0 0 0 1>}
 Pose_914 {type=Curve, loc=<0 0 0>, rot=<0 0 0 1>}
 Pose_915 {type=Curve, loc=<0 0 0>, rot=<0 0 0 1>}
 Pose_916 {type=Curve, loc=<0 0 0>, rot=<0 0 0 1>}
 Pose_917 {type=Curve, loc=<0 0 0>, rot=<0 0 0 1>}
 Pose_918 {type=Curve, loc=<0 0 0>, rot=<0 0 0 1>}
 Pose_919 {type=Curve, loc=<0 0 0>, rot=<0 0 0 1>}
 Pose_920 {type=Curve, loc=<0 0 0>, rot=<0 0 0 1>}
 Pose_921 {type=Curve, loc=<0 0 0>, rot=<0 0 0 1>}
 Pose_922 {type=Curve, loc=<0 0 0>, rot=<0 0 0 1>}
 Pose_923 {type=Curve, loc=<0 0 0>, rot=<0 0 0 1>}
 Pose_924 {type=Curve, loc=<0 0 0>, rot=<0 0 0 1>}
 Pose_925 {type=Curve, loc=<0 0 0>, rot=<0 0 0 1>}
 Pose_926 {type=Curve, loc=<0 0 0>, rot=<0 0 0 1>}
 Pose_927 {type=Curve, loc=<0 0 0>, rot=<0 0 0 1>}
 Pose_928 {type=Curve, loc=<0 0 0>, rot=<0 0 0 1>}
 Pose_929 {type=Curve, loc=<0 0 0>, rot=<0 0 0 1>}
 Pose_930 {type=Curve, loc=<0 0 0>, rot=<0 0 0 1>}
 Pose_931 {type=Curve, loc=<0 0 0>, rot=<0 0 0 1>}
 Pose_932 {type=Curve, loc=<0 0 0>, rot=<0 0 0 1>}
 Pose_933 {type=Curve, loc=<0 0 0>, rot=<0 0 0 1>}
 Pose_934 {type=Curve, loc=<0 0 0>, rot=<0 0 0 1>}
 Pose_935 {type=Curve, loc=<0 0 0>, rot=<0 0 0 1>}
 Pose_936 {type=Curve, loc=<0 0 0>, rot=<0 0 0 1>}
 Pose_937 {type=Curve, loc=<0 0 0>, rot=<0 0 0 1>}
 Pose_938 {type=Curve, loc=<0 0 0>, rot=<0 0 0 1>}
 Pose_939 {type=Curve, loc=<0 0 0>, rot=<0 0 0 1>}
 Pose_940 {type=Curve, loc=<0 0 0>, rot=<0 0 0 1>}
 Pose_941 {type=Curve, loc=<0 0 0>, rot=<0 0 0 1>}
 Pose_942 {type=Curve, loc=<0 0 0>, rot=<0 0 0 1>}
 Pose_943 {type=Curve, loc=<0 0 0>, rot=<0 0 0 1>}
 Pose_944 {type=Curve, loc=<0 0 0>, rot=<0 0 0 1>}
 Pose_945 {type=Curve, loc=<0 0 0>, rot=<0 0 0 1>}
 Pose_946 {type=Curve, loc=<0 0 0>, rot=<0 0 0 1>}
 Pose_947 {type=Curve, loc=<0 0 0>, rot=<0 0 0 1>}
 Pose_948 {type=Curve, loc=<0 0 0>, rot=<0 0 0 1>}
 Pose_949 {type=Curve, loc=<0 0 0>, rot=<0 0 0 1>}
 Pose_950 {type=Curve, loc=<0 0 0>, rot=<0 0 0 1>}
 Pose_951 {type=Curve, loc=<0 0 0>, rot=<0 0 0 1>}
 Pose_952 {type=Curve, loc=<0 0 0>, rot=<0 0 0 1>}
 Pose_953 {type=Curve, loc=<0 0 0>, rot=<0 0 0 1>}
 Pose_954 {type=Curve, loc=<0 0 0>, rot=<0 0 0 1>}
 Pose_955 {type=Curve, loc=<0 0 0>, rot=<0 0 0 1>}
 Pose_956 {type=Curve, loc=<0 0 0>, rot=<0 0 0 1>}
 Pose_957 {type=Curve, loc=<0 0 0>, rot=<0 0 0 1>}
 Pose_958 {type=Curve, loc=<0 0 0>, rot=<0 0 0 1>}
 Pose_959 {type=Curve, loc=<0 0 0>, rot=<0 0 0 1>}
 Pose_960 {type=Curve, loc=<0 0 0>, rot=<0 0 0 1>}
 Pose_961 {type=Curve, loc=<0 0 0>, rot=<0 0 0 1>}
 is_face_bone_anim {type=Curve, loc=<0 0 0>, rot=<0 0 0 1>}
 FullHourglass {type=Curve, loc=<0 0 0>, rot=<0 0 0 1>}
 LeanRectangle {type=Curve, loc=<0 0 0>, rot=<0 0 0 1>}
 InvertedTriangle {type=Curve, loc=<0 0 0>, rot=<0 0 0 1>}
 Pose_962 {type=Curve, loc=<0 0 0>, rot=<0 0 0 1>}
 Pose_963 {type=Curve, loc=<0 0 0>, rot=<0 0 0 1>}
 Pose_964 {type=Curve, loc=<0 0 0>, rot=<0 0 0 1>}
 Pose_965 {type=Curve, loc=<0 0 0>, rot=<0 0 0 1>}
 Pose_966 {type=Curve, loc=<0 0 0>, rot=<0 0 0 1>}
 Pose_967 {type=Curve, loc=<0 0 0>, rot=<0 0 0 1>}
 Pose_968 {type=Curve, loc=<0 0 0>, rot=<0 0 0 1>}
 Pose_969 {type=Curve, loc=<0 0 0>, rot=<0 0 0 1>}
 Pose_970 {type=Curve, loc=<0 0 0>, rot=<0 0 0 1>}
 Pose_971 {type=Curve, loc=<0 0 0>, rot=<0 0 0 1>}
 Pose_972 {type=Curve, loc=<0 0 0>, rot=<0 0 0 1>}
 Pose_973 {type=Curve, loc=<0 0 0>, rot=<0 0 0 1>}
 Pose_974 {type=Curve, loc=<0 0 0>, rot=<0 0 0 1>}
 Pose_975 {type=Curve, loc=<0 0 0>, rot=<0 0 0 1>}
 Pose_976 {type=Curve, loc=<0 0 0>, rot=<0 0 0 1>}
 Pose_977 {type=Curve, loc=<0 0 0>, rot=<0 0 0 1>}
 Pose_978 {type=Curve, loc=<0 0 0>, rot=<0 0 0 1>}
 Pose_979 {type=Curve, loc=<0 0 0>, rot=<0 0 0 1>}
 Pose_980 {type=Curve, loc=<0 0 0>, rot=<0 0 0 1>}
 Pose_981 {type=Curve, loc=<0 0 0>, rot=<0 0 0 1>}
 Pose_982 {type=Curve, loc=<0 0 0>, rot=<0 0 0 1>}
 Pose_983 {type=Curve, loc=<0 0 0>, rot=<0 0 0 1>}
 Pose_984 {type=Curve, loc=<0 0 0>, rot=<0 0 0 1>}
 Pose_985 {type=Curve, loc=<0 0 0>, rot=<0 0 0 1>}
 Pose_986 {type=Curve, loc=<0 0 0>, rot=<0 0 0 1>}
 Pose_987 {type=Curve, loc=<0 0 0>, rot=<0 0 0 1>}
 Pose_988 {type=Curve, loc=<0 0 0>, rot=<0 0 0 1>}
 Pose_989 {type=Curve, loc=<0 0 0>, rot=<0 0 0 1>}
 Pose_990 {type=Curve, loc=<0 0 0>, rot=<0 0 0 1>}
 Pose_991 {type=Curve, loc=<0 0 0>, rot=<0 0 0 1>}
 Pose_992 {type=Curve, loc=<0 0 0>, rot=<0 0 0 1>}
 Pose_993 {type=Curve, loc=<0 0 0>, rot=<0 0 0 1>}
 Pose_994 {type=Curve, loc=<0 0 0>, rot=<0 0 0 1>}
 Pose_995 {type=Curve, loc=<0 0 0>, rot=<0 0 0 1>}
 Pose_996 {type=Curve, loc=<0 0 0>, rot=<0 0 0 1>}
 Pose_997 {type=Curve, loc=<0 0 0>, rot=<0 0 0 1>}
 Pose_998 {type=Curve, loc=<0 0 0>, rot=<0 0 0 1>}
 Pose_999 {type=Curve, loc=<0 0 0>, rot=<0 0 0 1>}
 Pose_1000 {type=Curve, loc=<0 0 0>, rot=<0 0 0 1>}
 Pose_1001 {type=Curve, loc=<0 0 0>, rot=<0 0 0 1>}
 Pose_1002 {type=Curve, loc=<0 0 0>, rot=<0 0 0 1>}
 Pose_1003 {type=Curve, loc=<0 0 0>, rot=<0 0 0 1>}
 L_eye {type=Curve, loc=<0 0 0>, rot=<0 0 0 1>}
 R_eye {type=Curve, loc=<0 0 0>, rot=<0 0 0 1>}
 head {type=Curve, loc=<0 0 0>, rot=<0 0 0 1>}
 dyn_ear_l_master {type=Curve, loc=<0 0 0>, rot=<0 0 0 1>}
 root {type=Curve, loc=<0 0 0>, rot=<0 0 0 1>}
 dyn_ear02_l {type=Curve, loc=<0 0 0>, rot=<0 0 0 1>}
 dyn_ear03_l {type=Curve, loc=<0 0 0>, rot=<0 0 0 1>}
 dyn_ear04_l {type=Curve, loc=<0 0 0>, rot=<0 0 0 1>}
 dyn_ear05_l {type=Curve, loc=<0 0 0>, rot=<0 0 0 1>}
 dyn_ear01_l {type=Curve, loc=<0 0 0>, rot=<0 0 0 1>}
 flipbook_face_L_eye_index {type=Curve, loc=<0 0 0>, rot=<0 0 0 1>}
 flipbook_face_R_eye_index {type=Curve, loc=<0 0 0>, rot=<0 0 0 1>}
 flipbook_face_L_brow_index {type=Curve, loc=<0 0 0>, rot=<0 0 0 1>}
 flipbook_face_R_brow_index {type=Curve, loc=<0 0 0>, rot=<0 0 0 1>}
 flipbook_face_mouth_index {type=Curve, loc=<0 0 0>, rot=<0 0 0 1>}
 flipbook_face_bypass_mapping {type=Curve, loc=<0 0 0>, rot=<0 0 0 1>}
 neck_01_ctrl {controlType=EULER_TRANSFORM, type=Control, loc=<0 0 0>, rot=<0 0 0 1>}
 neck_02_ctrl {controlType=EULER_TRANSFORM, parent=neck_01_ctrl, type=Control, loc=<0 0 0>, rot=<0 0 0 1>}
 head_ctrl {controlType=EULER_TRANSFORM, parent=neck_02_ctrl, type=Control, loc=<0 0 0>, rot=<0 0 0 1>}
 coverHead_pv_ctrl {controlType=EULER_TRANSFORM, type=Control, loc=<0 0 0>, rot=<0 0 0 1>}
]
