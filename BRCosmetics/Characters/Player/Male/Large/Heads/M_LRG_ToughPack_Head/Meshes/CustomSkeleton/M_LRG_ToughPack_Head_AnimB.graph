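# Animation Blueprint for the large male ToughPack head cosmetic, posing two top-lid bones and adding secondary motion.
# The AnimGraph chains lid bone replacements, a head Control Rig, then spring-driven Anim Dynamics on three belt tip bones and a back handle, with one spherical collision limit.
# Each update computes a gravity override through a deprecated function and, when the body mesh has suitable bones, caches its lid transforms in parent-bone space.
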AnimBlueprint /BRCosmetics/Characters/Player/Male/Large/Heads/M_LRG_ToughPack_Head/Meshes/CustomSkeleton/M_LRG_ToughPack_Head_AnimB
target_skeleton=&/Fortnite_Base_Head_ToughPack_Skeleton
graphs [
 AnimGraph {nodes [
   n1 = "Transform (Modify) Bone - Bone: base_topLid"!(ComponentPose=Local To Component(LocalPose=Input Pose()), bAlphaBoolEnabled:bool=True, Translation="0,0,0", Rotation="0,0,0") -> (Pose) {
    alphaInputType=Bool
    rotationMode=BMM_Replace
    rotationSpace=BCS_ParentBoneSpace
    translationMode=BMM_Replace
    translationSpace=BCS_ParentBoneSpace
    alphaBoolBlend {blendOption=Linear}
    boneToModify {boneName=base_topLid}
   }
   n2 = "Transform (Modify) Bone - Bone: dyn_topLid"!(ComponentPose=@n1, bAlphaBoolEnabled:bool=True, Translation="0,0,0", Rotation="0,0,0") -> (Pose) {
    alphaInputType=Bool
    rotationMode=BMM_Replace
    rotationSpace=BCS_ParentBoneSpace
    translationMode=BMM_Replace
    translationSpace=BCS_ParentBoneSpace
    alphaBoolBlend {blendOption=Linear}
    boneToModify {boneName=dyn_topLid}
   }
   n3 = Control Rig(Source=Component To Local(ComponentPose=@n2)) {
    alphaBoolBlend {blendOption=Linear}
    controlRigAssetReference {blueprintRigClass=&/M_LRG_ToughPack_Head_CtrlRig.M_LRG_ToughPack_Head_CtrlRig_C}
   }
   n4 = Anim Dynamics - Bone: dyn_coverTip_belt_l_1!(ComponentPose=Local To Component(LocalPose=@n3), Alpha=1, GravityOverride:Vector="0,0,-980") -> (Pose) {
    angularDampingOverride=1
    angularSpringConstant=6000
    bAngularSpring=true
    bOverrideAngularDamping=true
    bUseGravityOverride=true
    gravityScale=0.35
    simulationSpace=RootRelative
    alphaBoolBlend {blendOption=Linear}
    boundBone {boneName=dyn_coverTip_belt_l_1}
    componentAppliedLinearAccClamp=<3000 3000 3000>
    componentLinearAccScale=<1 1 1>
    componentLinearVelScale=<1 1 1>
    gravityOverride {z=-980}
    physicsBodyDefinitions [{
      boundBone {boneName=dyn_coverTip_belt_l_1}
      boxExtents=<20 30 30>
      constraintSetup {angularLimitsMax=<0 0 20>, angularLimitsMin=<0 0 0>, angularTarget=<1 0 0>}
      localJointOffset=<8 0 0>
     }]
   }
   n5 = Anim Dynamics - Bone: dyn_coverTip_belt_r_1!(ComponentPose=@n4, Alpha=1, GravityOverride:Vector="0,0,-980") -> (Pose) {
    angularDampingOverride=1
    angularSpringConstant=6000
    bAngularSpring=true
    bOverrideAngularDamping=true
    bUseGravityOverride=true
    bUseSphericalLimits=true
    gravityScale=0.35
    simulationSpace=RootRelative
    alphaBoolBlend {blendOption=Linear}
    boundBone {boneName=dyn_coverTip_belt_r_1}
    componentAppliedLinearAccClamp=<3000 3000 3000>
    componentLinearAccScale=<1 1 1>
    componentLinearVelScale=<1 1 1>
    gravityOverride {z=-980}
    physicsBodyDefinitions [{
      boundBone {boneName=dyn_coverTip_belt_r_1}
      boxExtents=<20 30 30>
      constraintSetup {angularLimitsMax=<0 0 20>, angularLimitsMin=<0 0 0>, angularTarget=<1 0 0>}
      localJointOffset=<8 0 0>
     }]
    sphericalLimits [{limitRadius=13.0918, drivingBone {boneName=spine_04}, sphereLocalOffset=<4.7184 -2.0308 8.7825>}]
   }
   n6 = Anim Dynamics - Bone: dyn_coverTip_belt_r_2!(ComponentPose=@n5, Alpha=1, GravityOverride:Vector="0,0,-980") -> (Pose) {
    angularDampingOverride=1
    angularSpringConstant=6000
    bAngularSpring=true
    bOverrideAngularDamping=true
    bUseGravityOverride=true
    gravityScale=0.35
    simulationSpace=RootRelative
    alphaBoolBlend {blendOption=Linear}
    boundBone {boneName=dyn_coverTip_belt_r_2}
    componentAppliedLinearAccClamp=<3000 3000 3000>
    componentLinearAccScale=<1 1 1>
    componentLinearVelScale=<1 1 1>
    gravityOverride {z=-980}
    physicsBodyDefinitions [{
      boundBone {boneName=dyn_coverTip_belt_r_2}
      boxExtents=<20 30 30>
      constraintSetup {angularLimitsMax=<0 0 20>, angularLimitsMin=<0 0 0>, angularTarget=<1 0 0>}
      localJointOffset=<8 0 0>
     }]
   }
   n7 = Anim Dynamics - Bone: dyn_backHandle!(ComponentPose=@n6, Alpha=1, GravityOverride:Vector="0,0,-980") -> (Pose) {
    angularDampingOverride=1
    angularSpringConstant=6000
    bAngularSpring=true
    bOverrideAngularDamping=true
    bUseGravityOverride=true
    gravityScale=0.35
    simulationSpace=RootRelative
    alphaBoolBlend {blendOption=Linear}
    boundBone {boneName=dyn_backHandle}
    componentAppliedLinearAccClamp=<3000 3000 3000>
    componentLinearAccScale=<1 1 1>
    componentLinearVelScale=<1 1 1>
    gravityOverride {z=-980}
    physicsBodyDefinitions [{
      boundBone {boneName=dyn_backHandle}
      boxExtents=<20 20 20>
      constraintSetup {angularLimitsMax=<0 0 0>, angularLimitsMin=<0 0 -60>, angularTarget=<1 0 0>}
      localJointOffset=<5 0 0>
     }]
   }
   Output Pose(Result=Component To Local(ComponentPose=@n7))
  ]}
 EventGraph {nodes [
   Event BlueprintUpdateAnimation()
   > n1 = Sequence()
   > n2 = Gravity Override State Function DEPRECATED(CharacterPartAnimInstance=Self-Reference(), GameTimeInSeconds=GetGameTimeInSeconds(), GravityOverrideParams=Get Gravity Override Params())
   > Set GravityOverride(GravityOverride=@n2)
   n3 = Get BodyMesh() -> (BodyMesh:Skeletal Mesh Component Object Reference)
   Is Valid(exec=@n1.then_1, InputObject=@n3)
   n4 = GetSocketBoneName!(self:Skinned Mesh Component Object Reference=@n3, InSocketName:Name=upperarm_r) -> (ReturnValue:Name=None)
   > n5 = Set isUsingBodyBones(isUsingBodyBones:bool="Not Equal Exactly (String)"(A="To String (Name)"(InName=@n4), B=None)) -> (Output_Get:bool)
   > Branch(Condition=@n5)
   n6 = Get BodyMesh()
   n7 = BreakTransform(InTransform=GetSocketTransform(self=@n6, InSocketName=base_topLid, TransformSpace=RTS_ParentBoneSpace))
   > Set base_topLid_location(base_topLid_location:Vector=@n7) -> (Output_Get:Vector)
   > Set base_topLid_rotation(base_topLid_rotation:Rotator=@n7.Rotation) -> (Output_Get:Rotator)
   n8 = BreakTransform(InTransform=GetSocketTransform(self=@n6, InSocketName=dyn_topLid, TransformSpace=RTS_ParentBoneSpace))
   > Set dyn_topLid_location(dyn_topLid_location:Vector=@n8) -> (Output_Get:Vector)
   > Set dyn_topLid_rotation(dyn_topLid_rotation:Rotator=@n8.Rotation) -> (Output_Get:Rotator)
   n9 = Event OnCharacterCustomizationLoaded()
   > Set BodyMesh(BodyMesh=Get Array Item with EFort Custom Part Type(Array=@n9.CustomMeshParts))
  ]}
]
variables [
 base_topLid_location {type=Vector}
 base_topLid_rotation {type=Rotator}
 BodyMesh {type=SkeletalMeshComponent}
 dyn_topLid_location {type=Vector}
 dyn_topLid_rotation {type=Rotator}
 Gravity Override Params {type=GravityOverrideParamsStruct}
 GravityOverride {type=Vector}
 isUsingBodyBones {type=boolean}
]
